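# Translucent unlit Niagara ribbon and sprite material for the dirt trail in the InflatoDodo skydiving contrail.
# Panning wind-noise and grunge textures, offset by a NoiseRandom dynamic parameter, feed named reroutes for distortion and breakup.
# Distorted UVs build a sine-tapered streak mask, combined with edge fades, particle alpha and an Erosion threshold for opacity.
# Particle color drives noise-modulated emissive, and CamOffset drives camera-offset world position.
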Material /BRCosmetics/Athena/Effects/Particles/Skydiving/NS_Contrail_InflatoDodo/M_InflatoDodo_Contrail_DirtTrail
graph {
 properties {
  bAutomaticallySetUsageInEditor=false
  bUsedWithNiagaraRibbons=true
  bUsedWithNiagaraSprites=true
  blendMode=BLEND_Translucent
  shadingModel=MSM_Unlit
  usedShadingModels=Unlit
 }
 nodes [
  MaterialFunctionCall(Offset Amount=DynamicParameter() {defaultValue=<rgba 0 0 1 1>, paramNames [NoiseRandom,CamOffset,Param3,Param4]}.CamOffset) {materialFunction=/Engine/Functions/Engine_MaterialFunctions02/WorldPositionOffset/CameraOffset}
  n1 = TextureCoordinate()
  n2 = NamedRerouteUsage()
  n3 = Multiply(A=@n1, B=Power(Base=@n2) {constExponent=1})
  n4 = Constant2Vector() {r=2}
  n5 = MaterialFunctionCall(Float2=Add(A=@n1, B=Multiply(A=@n3, B=@n4))) {materialFunction=&BreakOutFloat2Components}
  n6 = Sine(None=Multiply(A=@n5.G) {constB=0.5})
  n7 = OneMinus(None=Multiply(A=Multiply(A=@n5.R) {constB=2}, B=OneMinus(None=@n6)))
  n8 = Power(Base=@n7) {constExponent=10}
  Max(A=Multiply(A=@n8) {constB=0.2}, B=Power(Base=@n7) {constExponent=20})
  n9 = TextureCoordinate()
  n10 = MaterialFunctionCall(Float2=@n9) {materialFunction=&BreakOutFloat2Components}
  n11 = Power(Base=@n10.R) {constExponent=1.5}
  n12 = NamedRerouteUsage()
  n13 = Max(A=@n2, B=Power(Base=Multiply(A=@n12, B=@n11)) {constExponent=0.5})
  n14 = Add(A=Multiply(A=Multiply(A=@n13, B=@n10.R)), B=@n10.R)
  n15 = OneMinus(None=@n14)
  Multiply(A=@n13, B=@n15)
  n16 = Saturate(None=Multiply(A=@n12) {constB=5})
  n17 = MaterialFunctionCall(Float2=TextureCoordinate()) {materialFunction=&BreakOutFloat2Components}
  Multiply(
   A=MaterialFunctionCall(Emissive=ParticleColor().RGB) {materialFunction=&MF_FX_EmissiveAdjustments}.Result
   B=LinearInterpolate(A=LinearInterpolate(Alpha=@n2) {constA=0.4}, B=LinearInterpolate(Alpha=@n16) {constA=1, constB=0.4}, Alpha=Multiply(A=@n17.R)) {constA=1}
  )
  n18 = Multiply(A=Power(Base=@n5.R) {constExponent=1})
  n19 = Power(Base=Multiply(A=Saturate(None=@n8), B=OneMinus(None=@n18))) {constExponent=1}
  n20 = ComponentMask(None=TextureCoordinate()) {r=true}
  n21 = OneMinus(None=@n20)
  n22 = Saturate(None=Multiply(A=Power(Base=@n12) {constExponent=0.8}) {constB=5})
  n23 = Power(Base=Multiply(A=@n20))
  n24 = OneMinus(None=LinearInterpolate(B=Max(A=@n21, B=@n22), Alpha=Saturate(None=@n23)) {constA=1})
  n25 = Saturate(None=Multiply(A=Subtract(A=@n19, B=@n24)))
  n26 = Power(Base=@n17.R)
  n27 = Sine(None=Multiply(A=@n17.G) {constB=0.5})
  n28 = Saturate(None=Multiply(A=Multiply(A=@n26) {constB=12}, B=Multiply(A=@n27) {constB=3}))
  n29 = NamedRerouteUsage()
  n30 = LinearInterpolate(Alpha=ScalarParameter() {parameterName=Erosion}) {constA=-1}
  Multiply(A=Multiply(A=Multiply(A=@n25, B=@n28), B=ParticleColor().A), B=Saturate(None=Subtract(A=@n29, B=@n30)))
  n31 = Add(A=DynamicParameter() {defaultValue=<rgba 0 0 1 1>, paramNames [NoiseRandom,CamOffset,Param3,Param4]}.NoiseRandom, B=@n9)
  n32 = Multiply(A=Constant2Vector() {g=2, r=1.1}, B=@n31)
  NamedRerouteDeclaration(None=TextureSampleParameter2D(UVs=Panner(Coordinate=@n32) {speedX=-2, speedY=0.21}) {parameterName=Wind, samplerType=SAMPLERTYPE_Grayscale, texture=/Game/Global/Textures/Default/Noise/TilingNoise06}.B) {name=WindTexture, nodeColor=<rgba 0.0312 1 0 1>}
  n33 = Constant2Vector() {g=1, r=0.8}
  n34 = TextureSampleParameter2D(UVs=Panner(Coordinate=Multiply(A=@n31, B=@n33)) {speedX=-0.8, speedY=-0.12}) {parameterName=Grunge, samplerType=SAMPLERTYPE_LinearGrayscale, texture=/Game/Effects/Fort_Effects/Textures/Rocks/T_Tiling_Dust_Dirt_Bits_M}
  NamedRerouteDeclaration(None=@n34.G) {name=GrungeTexture, nodeColor=<rgba 1 0 0.3516 1>}
  n35 = Power(Base=@n15) {constExponent=5}
  n36 = Power(Base=@n14) {constExponent=0.1}
  Saturate(None=Divide(A=Min(A=@n35, B=@n36)) {constB=1})
 ]
}
