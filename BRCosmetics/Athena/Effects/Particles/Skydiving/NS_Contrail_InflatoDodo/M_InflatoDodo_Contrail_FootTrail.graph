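# Translucent unlit, two-sided Niagara mesh particle material for the foot trail of the InflatoDodo skydiving contrail.
# Panning flame noise, offset per particle, blends red to orange emissive, darkened along UV V.
# Opacity combines that noise with end fades and inverse Fresnel for soft edges, with particle alpha driving erosion.
# Dynamic parameters TwistDirection and TwistIntensity twist vertices around the local X axis, weighted along the mesh's V coordinate.
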Material /BRCosmetics/Athena/Effects/Particles/Skydiving/NS_Contrail_InflatoDodo/M_InflatoDodo_Contrail_FootTrail
graph {
 properties {
  bAutomaticallySetUsageInEditor=false
  bUsedWithNiagaraMeshParticles=true
  blendMode=BLEND_Translucent
  shadingModel=MSM_Unlit
  twoSided=true
  usedShadingModels=Unlit
 }
 nodes [
  Constant3Vector() {constant=<rgba -1 0 0 0>}
  Constant() {r=0.3}
  n1 = Constant3Vector() {constant=<rgba 0.05 0 0 0>}
  n2 = MaterialFunctionCall(Float2=TextureCoordinate()) {materialFunction=&BreakOutFloat2Components}
  n3 = Power(Base=OneMinus(None=@n2.G)) {constExponent=1}
  n4 = Constant3Vector() {constant=<rgba 1 0.0333 0 1>}
  n5 = Constant3Vector() {constant=<rgba 1.3 0.3473 0.0723 1>}
  n6 = Multiply(A=TextureCoordinate(), B=Constant2Vector() {g=0.1, r=3})
  n7 = TextureSampleParameter2D(UVs=Add(A=Panner(Coordinate=@n6) {speedY=0.5}, B=ParticleRandom())) {parameterName=TexNoise, texture=/BRCosmetics/Characters/Player/Male/Medium/Bodies/M_MED_VelvetDeskFiber/Textures/T_FlameNoise}
  n8 = SmoothStep(Value=Multiply(A=Power(Base=@n2.G), B=Power(Base=@n7.R) {constExponent=1.5})) {constMax=0.08}
  MaterialFunctionCall(Emissive=Multiply(A=LinearInterpolate(B=@n1, Alpha=@n3) {constA=1}, B=LinearInterpolate(A=@n4, B=@n5, Alpha=@n8))) {materialFunction=&MF_FX_EmissiveAdjustments}
  n9 = Saturate(None=Multiply(A=@n8))
  n10 = OneMinus(None=Power(Base=Multiply(A=@n2.G)) {constExponent=6})
  n11 = LinearInterpolate(Alpha=ParticleColor().A) {constA=1, constB=-1}
  n12 = Fresnel() {baseReflectFraction=0, exponent=2}
  Multiply(A=Multiply(A=Multiply(A=@n9, B=@n10), B=Subtract(A=@n7.R, B=@n11)), B=Power(Base=OneMinus(None=@n12)) {constExponent=1})
  n13 = DynamicParameter() {paramNames [TwistDirection, TwistIntensity, Param3, Param4]}
  n14 = Constant()
  n15 = MaterialFunctionCall(R=@n13.TwistDirection, G=@n14, B=@n14) {materialFunction=&AppendMany}
  n16 = OneMinus(None=MaterialFunctionCall(Float2=TextureCoordinate()) {materialFunction=&BreakOutFloat2Components}.G)
  n17 = Constant() {r=1}
  RotateAboutAxis(
   NormalizedRotationAxis=Normalize(VectorInput=Transform(None=@n15.RGB) {transformSourceType=TRANSFORMSOURCE_Local})
   RotationAngle=Multiply(A=Power(Base=@n16) {constExponent=1}, B=Multiply(A=@n13.TwistIntensity, B=@n17)) {constA=1}
   PivotPoint=TransformPosition(None=Constant3Vector()) {transformType=TRANSFORMPOSSOURCE_World}
   Position=WorldPosition().XYZ
  )
 ]
}
