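# Masked, two-sided lit material for footprint particles in the InflatoDodo skydiving contrail, built for Niagara mesh and sprite particles.
# UV gradients and a skin texture blend foot and finger colors, with vertex color masking in claw colors.
# The BottomFlames dynamic parameter fades in panning-noise flames for color and emissive, while ClawAnimWPO drives a UV-weighted pivot rotation for World Position Offset.
# Particle alpha drives a noise dissolve, randomly offset per particle.
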Material /BRCosmetics/Athena/Effects/Particles/Skydiving/NS_Contrail_InflatoDodo/M_InflatoDodo_Footprint
graph {
 properties {
  bAutomaticallySetUsageInEditor=false
  bUsedWithNiagaraMeshParticles=true
  bUsedWithNiagaraSprites=true
  blendMode=BLEND_Masked
  twoSided=true
  usedShadingModels=Default Lit
 }
 nodes [
  n1 = MaterialFunctionCall(Float2=TextureCoordinate()) {materialFunction=&BreakOutFloat2Components}
  n2 = LinearInterpolate(Alpha=@n1.R) {constA=0.4, constB=-0.2}
  n3 = Transform(None=MaterialFunctionCall(R=LinearInterpolate(Alpha=@n1.R), G=@n2, B=@n2) {materialFunction=&AppendMany}.RGB) {transformSourceType=TRANSFORMSOURCE_Local}
  n4 = Power(Base=OneMinus(None=@n1.G)) {constExponent=1.6}
  n5 = Multiply(A=Constant() {r=0.7}, B=DynamicParameter() {defaultValue=<rgba 0.5 1 1 1>, paramNames [ClawAnimWPO,BottomFlames,Param3,Param4]}.ClawAnimWPO)
  n6 = Constant3Vector() {constant=<rgba 0 -20 5 0>}
  NamedRerouteDeclaration(None=RotateAboutAxis(
    NormalizedRotationAxis=Normalize(VectorInput=@n3)
    RotationAngle=Multiply(A=@n4, B=@n5) {constA=1}
    PivotPoint=TransformPosition(None=@n6) {transformType=TRANSFORMPOSSOURCE_World}
    Position=WorldPosition().XYZ
   )) {name=WPO, nodeColor=<rgba 1 0.5391 0 1>}
  n7 = Multiply(A=TextureCoordinate()) {constB=2}
  n8 = ParticleRandom()
  NamedRerouteDeclaration(None=TextureSampleParameter2D(UVs=Add(A=@n7, B=@n8)) {parameterName=Dissolve, texture=/Game/Effects/Fort_Effects/Textures/Tile/T_CircleNoise_01_M}.R) {name=Dissolve Texture, nodeColor=<rgba 0 1 0.4375 1>}
  n9 = TextureSampleParameter2D(UVs=Multiply(A=TextureCoordinate()) {constB=2}) {parameterName=Param, samplerType=SAMPLERTYPE_Normal, texture=/Game/Effects/Fort_Effects/Textures/Tile/T_GaussianNoise_02_N}
  Reroute(None=Multiply(A=Power(Base=@n9.RGB) {constExponent=1}) {constB=50})
  Reroute(None=NamedRerouteUsage())
  n10 = NamedRerouteUsage()
  n11 = LinearInterpolate(Alpha=ParticleColor().A) {constA=1, constB=-1}
  n12 = OneMinus(None=@n11)
  n13 = Multiply(A=Divide(A=Subtract(A=@n10, B=@n11), B=Saturate(None=@n12))) {constB=30}
  Reroute(None=Multiply(A=Constant() {r=1}, B=Saturate(None=@n13)))
  n14 = MaterialFunctionCall(Float2=TextureCoordinate() {coordinateIndex=1}) {materialFunction=&BreakOutFloat2Components}
  n15 = Power(
   Base=@n14.G
   Exp=LinearInterpolate(Alpha=DynamicParameter() {defaultValue=<rgba 0.5 1 1 1>, paramNames [ClawAnimWPO,BottomFlames,Param3,Param4]}.BottomFlames) {constA=4, constB=2.5}
  ) {constExponent=3}
  n16 = Panner(Coordinate=TextureCoordinate() {coordinateIndex=1}) {speedY=3}
  n17 = Constant2Vector() {g=0.3, r=1}
  n18 = Multiply(A=TextureSampleParameter2D(UVs=Multiply(A=@n16, B=@n17)) {parameterName=TexFlame, texture=/BRCosmetics/Characters/Player/Male/Medium/Bodies/M_MED_VelvetDeskFiber/Textures/T_FlameNoise}.R) {constB=2}
  n19 = Saturate(None=Multiply(A=@n15, B=Add(A=@n15, B=@n18)) {constA=1})
  n20 = LinearInterpolate(
   A=VectorParameter() {parameterName=FlameBaseColorB, defaultValue=<rgba 1 0.043 0.01 1>}.RGB
   B=VectorParameter() {parameterName=FlameBaseColorA, defaultValue=<rgba 2 0.575 0.1 1>}.RGB
   Alpha=@n19
  )
  n21 = LinearInterpolate(B=@n20, Alpha=@n19)
  n22 = DynamicParameter() {defaultValue=<rgba 0.5 1 1 1>, paramNames [ClawAnimWPO,BottomFlames,Param3,Param4]}
  Reroute(None=MaterialFunctionCall(Emissive=LinearInterpolate(B=@n21, Alpha=@n22.BottomFlames)) {materialFunction=&MF_FX_EmissiveAdjustments}.Result)
  Reroute(None=LinearInterpolate(Alpha=Power(Base=@n9.R)) {constA=1, constB=0})
  n23 = VectorParameter() {parameterName=FootColorBase, defaultValue=<rgba 0.012 0.02 0.004 1>}
  n24 = LinearInterpolate(Alpha=NamedRerouteUsage()) {constA=0.4}
  n25 = VectorParameter() {parameterName=FootColorMain, defaultValue=<rgba 0.3373 0.46 0.092 1>}
  n26 = Multiply(A=Power(Base=NamedRerouteUsage()) {constExponent=2.5}) {constB=2}
  n27 = TextureSampleParameter2D(UVs=Multiply(A=TextureCoordinate()) {constB=2}) {
   parameterName=LizardSkinTex
   samplerType=SAMPLERTYPE_LinearGrayscale
   texture=/Game/Effects/Fort_Effects/Textures/Tile/T_Caustic_Wind_Diagonal
  }
  n28 = LinearInterpolate(
   A=LinearInterpolate(
    A=Multiply(A=@n24, B=@n25.RGB)
    B=VectorParameter() {parameterName=FootColorHighlight, defaultValue=<rgba 0.6 0.42 0.06 1>}.RGB
    Alpha=Saturate(None=@n26)
   )
   B=Multiply(A=@n25.RGB) {constB=0.05}
   Alpha=Power(Base=@n27.R) {constExponent=0.8}
  )
  n29 = MaterialFunctionCall(Float2=TextureCoordinate() {coordinateIndex=1}) {materialFunction=&BreakOutFloat2Components}
  n30 = OneMinus(None=@n29.G)
  n31 = Saturate(None=Multiply(A=Power(Base=@n30) {constExponent=2.7}) {constB=3})
  n32 = OneMinus(None=Multiply(A=@n29.G))
  n33 = Saturate(None=Multiply(A=Power(Base=@n32) {constExponent=9}))
  n34 = Saturate(None=Multiply(A=Power(Base=@n32) {constExponent=12.5}))
  n35 = Multiply(A=Power(Base=Multiply(A=@n29.R) {constB=2}) {constExponent=6}, B=SmoothStep(Value=Subtract(A=@n33, B=@n34)) {constMax=0.05})
  n36 = LinearInterpolate(
   A=LinearInterpolate(
    A=LinearInterpolate(A=@n23.RGB, B=@n28, Alpha=@n31)
    B=VectorParameter() {parameterName=FingerHighlight, defaultValue=<rgba 0.523 0.55 0.022 1>}.RGB
    Alpha=Power(Base=@n35) {constExponent=1}
   )
   B=LinearInterpolate(
    A=VectorParameter() {parameterName=ClawColorBase, defaultValue=<rgba 0.005 0.005 0.005 0>}.RGB
    B=VectorParameter() {parameterName=ClawColorHighlight, defaultValue=<rgba 0.025 0.021 0.02 0>}.RGB
    Alpha=Power(Base=@n31) {constExponent=0.8}
   )
   Alpha=VertexColor()
  )
  n37 = LinearInterpolate(A=@n36, B=@n20, Alpha=@n19)
  Reroute(None=Multiply(A=LinearInterpolate(A=@n36, B=@n37, Alpha=@n22.BottomFlames), B=ParticleColor().RGB))
  Saturate(None=Power(Base=Multiply(A=@n14.G)) {constExponent=8})
 ]
}
