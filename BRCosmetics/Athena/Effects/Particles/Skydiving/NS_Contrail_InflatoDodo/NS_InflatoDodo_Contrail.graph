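# Niagara skydiving contrail cosmetic for InflatoDodo, with dirt clouds, dirt ribbons and timed fiery footstep impacts.
# Clouds and four ribbons follow sockets from a shared Socket Reader, with ribbons procedurally extended along a trail vector.
# FootprintLeft uses spawn groups and visibility tags to drive footprint, chunk, impact, flame and spark renderers from one emitter.
# System dive values, front-end modifiers and scalability multipliers adjust motion, size and spawn counts.
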NiagaraSystem /BRCosmetics/Athena/Effects/Particles/Skydiving/NS_Contrail_InflatoDodo/NS_InflatoDodo_Contrail
dependencies {
 usedDataInterfaces [
  &NiagaraDataInterfaceArrayFloat
  &NiagaraDataInterfaceArrayFloat3
  &NiagaraDataInterfaceCurve
  &NiagaraDataInterfaceColorCurve
  &NiagaraDataInterfaceSpriteRendererInfo
  &NiagaraDataInterfaceMeshRendererInfo
  &NiagaraDataInterfaceSocketReader
  &NiagaraDataInterfacePlatformSet
 ]
 usedDynamicInputs [
  &FN_PickFloatDependOnPlatform
  &TransformVector
  &NormalizeVector
  &TransformPosition
 ]
 usedModules [
  &self:SystemSpawnScript.SystemScriptSource.SystemScriptGraph.NiagaraNodeAssignment_2.SetVariables_#1
  &self:SystemSpawnScript.SystemScriptSource.SystemScriptGraph.NiagaraNodeAssignment_9.SetVariables_#2
  &self:SystemSpawnScript.SystemScriptSource.SystemScriptGraph.NiagaraNodeAssignment_7.SetVariables_#3
  &self:SystemSpawnScript.SystemScriptSource.SystemScriptGraph.NiagaraNodeAssignment_10.SetVariables_#4
  &SystemState
  &self:SystemSpawnScript.SystemScriptSource.SystemScriptGraph.NiagaraNodeAssignment_3.SetVariables_#5
  &self:SystemSpawnScript.SystemScriptSource.SystemScriptGraph.NiagaraNodeAssignment_4.SetVariables_#6
  &self:SystemSpawnScript.SystemScriptSource.SystemScriptGraph.NiagaraNodeAssignment_5.SetVariables_#7
  &self:SystemSpawnScript.SystemScriptSource.SystemScriptGraph.NiagaraNodeAssignment_0.SetVariables_#8
  &self:SystemSpawnScript.SystemScriptSource.SystemScriptGraph.NiagaraNodeAssignment_1.SetVariables_#9
  &self:SystemSpawnScript.SystemScriptSource.SystemScriptGraph.NiagaraNodeAssignment_6.SetVariables_#10
  &self:SystemSpawnScript.SystemScriptSource.SystemScriptGraph.NiagaraNodeAssignment_8.SetVariables_#11
 ]
}
emitters {
 DirtClouds {
  emitterUpdateScript [
   &EmitterState(
    Inactive Response=::NewEnumerator1
    Life Cycle Mode=::NewEnumerator1
    Loop Behavior=::NewEnumerator1
    Loop Duration Mode=::NewEnumerator1
    UseLoopDelay=false
   )
   &SpawnRate(SpawnRate=14, Use Spawn Probability=false)
   &SpawnBurst_Instantaneous(Spawn Count=4, Spawn Probability=HlslExpression() {hlslExpression="Emitter.LoopCount == 0"}, Use Spawn Probability=false) {enabled=false, module=SpawnBurst_Instantaneous001}
  ]
  particleSpawnScript [
   &FN_SetSpawnGroup()
   &PartitionParticles(Number Of Partitions=4)
   &InitializeParticle(
    Color=<rgba 0.05 0.035 0.03 1>
    Lifetime=&Multiply_Float(
     A=&RandomRangeFloat(Minimum=1.3, Maximum=1.7)
     B=&FN_FrontEndFloatModifier(Gameplay=1, Front End=&Lerp_Float(A=1, B=0.8, Alpha=$System.DiveSprint))
    )
    Sprite Rotation Mode=::NewEnumerator2
    Sprite Size Mode=::NewEnumerator3
    Uniform Sprite Size=&Multiply_Float(
     A=&RandomRangeFloat(Minimum=35, Maximum=45)
     B=&MergeHelper_FloatSwitch(0=1, 1=1, 2=0.7, 3=0.7, Index=$Particles.PartitionParticles.Partition, Num=::NewEnumerator2)
    )
    UsePositionOffset=false
    Write Material Random=true
   ) {module=InitializeParticle001}
   &CameraOffset()
  ]
  particleUpdateScript [
   &SocketLocation(
    Orientation Sampling=::NewEnumerator1
    SamplePosOffset=true
    Sampled Position Offset=&Multiply_VectorByFloat(
     Vector=&FN_FrontEndVector3Modifier(Gameplay=<0 0 40>, Front End=&Lerp_Vector(A=<0 0 40>, B=<0 -20 45>, Alpha=$System.DiveSprint))
     Float=&FloatFromCurve(
      FloatCurve=DataInterface() {propertyValues {Curve {keys [<at 0 -0.2>,<at 1 1>]}}}
      Scale Curve=&RandomRangeFloat(Minimum=1, Maximum=2)
     )
    )
    Sampled Position Offset Coordinate Space:ENiagaraSkelMeshTransforms=::NewEnumerator2
    Socket Index=$Particles.PartitionParticles.Partition
    Socket Reader=$System.Socket Reader
    Socket Sampling Mode=::NewEnumerator3
    VelocitySpeedLimit=false
   )
   &ParticleState()
   &ScaleColor(
    Scale Alpha=&FloatFromCurve(
     FloatCurve=DataInterface() {propertyValues {Curve {keys [
         {interpMode=RCIM_Cubic, tangentMode=RCTM_User, time=0, value=0}
         {interpMode=RCIM_Cubic, tangentMode=RCTM_User, time=0.4, value=1}
         <at 1 0 RCIM_Cubic>
        ]}}}
     Scale Curve=&Multiply_Float(
      A=&RandomRangeFloat(Minimum=0.9, Maximum=1.5)
      B=&FN_FrontEndFloatModifier(Gameplay=1, Front End=0.75)
     )
    )
    Scale RGB=&VectorFromFloat(Value=&FloatFromCurve(FloatCurve=DataInterface() {propertyValues {Curve {keys [<at 0 1>,<at 1 0>]}}}))
    ScaleA=true
    ScaleRGB=true
   )
   /SetVariables(Particles.SpriteSize=&Multiply_Vector2D_ByFloat(
     Float=&Multiply_Float(
      A=1
      B=&FloatFromCurve(FloatCurve=DataInterface() {propertyValues {Curve {keys [{interpMode=RCIM_Cubic, tangentMode=RCTM_User, time=0, value=0.1}, {interpMode=RCIM_Cubic, tangentMode=RCTM_User, time=0.35, value=1}]}}})
     )
     Vector2D=$Particles.Initial.SpriteSize
    )) {isSetParameters=true}
   &SubUVAnimation(
    "Sprite Renderer "=DataInterface() {propertyValues {SpriteRenderer=&self:DirtClouds_0.NiagaraSpriteRendererProperties_1}}
    UseEndFrame=false
    UseStartFrame=false
   ) {enabled=false}
   /SetVariables(Particles.ColorDirt:LinearColor=&MakeLinearColorFromVectorAndFloat(
     "Vector (RGB)"=&MakeVectorFromLinearColorRGB(Color=&RandomRangeLinearColor(Minimum=<rgba 0.12 0.088 0.072 1>, Maximum=<rgba 0.07 0.0408 0.035 1>))
     "Float (Alpha)"=&FloatFromCurve(FloatCurve=DataInterface() {propertyValues {Curve {keys [
          {interpMode=RCIM_Cubic, tangentMode=RCTM_User, time=0, value=0}
          {interpMode=RCIM_Cubic, tangentMode=RCTM_User, time=0.2, value=1}
          {arriveTangent=-0.0434, interpMode=RCIM_Cubic, leaveTangent=-0.0434, tangentMode=RCTM_User, time=0.3, value=1}
          {interpMode=RCIM_Cubic, tangentMode=RCTM_User, time=0.7, value=0}
         ]}}})
    )) {isSetParameters=true}
   /SetVariables(Particles.SpriteSizeDirt:Vector2f=&Multiply_Vector2D_ByFloat(Float=1.2, Vector2D=$Particles.SpriteSize)) {isSetParameters=true}
   &DynamicMaterialParameters(
    Index 0 Param 1=&FloatFromCurve(FloatCurve=DataInterface() {propertyValues {Curve {keys [<at 0 1>,<at 1 0>]}}})
    Module.Index0Float0Write=true
    Module.Index0Float1Write=false
    Module.Index0Float2Write=false
    Module.Index0Float3Write=false
   ) {enabled=false}
  ]
  rendererClasses [&NiagaraSpriteRendererProperties]
  renderers [
   {rendererClass=&NiagaraSpriteRendererProperties}
   {rendererClass=&NiagaraSpriteRendererProperties, rendererIndex=1}
  ]
 }
 Dirt_Ribbons {
  emitterSpawnScript [
   /SetVariables(Emitter.InterpolatedVector:Vector3f=<0 0 1>, Emitter.TargetVector:Vector3f=<0 0 1>) {isSetParameters=true}
   /SetVariables(Emitter.RibbonLength=222) {isSetParameters=true}
   /SetVariables(Emitter.ParticlesPerTrail=5) {isSetParameters=true}
  ]
  emitterUpdateScript [
   &EmitterState(
    Inactive Response=::NewEnumerator1
    Life Cycle Mode=::NewEnumerator1
    Loop Behavior=::NewEnumerator1
    Loop Duration Mode=::NewEnumerator1
    UseLoopDelay=false
   )
   &SpawnBurst_Instantaneous(
    Spawn Count=&Multiply_Int(A=4, B=$Emitter.ParticlesPerTrail)
    Use Spawn Probability=false
    UseLoopCountLimit=false
   )
   /SetVariables(Emitter.InterpolatedVector:Vector3f=&FN_FrontEndVector3Modifier(
     Gameplay=<0 0 1>
     Front End=/Game/Effects/Niagara/DynamicInputs/FrontEnd/FN_FrontEndRotateVector(Angle:NiagaraFloat=&Lerp_Float(A=0, B=0.4, Alpha=$System.ContrailDiveAlpha), Vector to Rotate:Vector3f=<0 0 1>)
    )) {isSetParameters=true}
   &self:ScratchModule(Target Vector:Vector3f=$System.LookAtVector) {enabled=false, module=ScratchModule}
  ]
  particleSpawnScript [
   &InitializeParticle(
    Lifetime=2
    Position Mode=::NewEnumerator0
    Ribbon Width=15
    Ribbon Width Mode=::NewEnumerator1
    Sprite Size Mode=::NewEnumerator3
    Uniform Sprite Size=5
    Write Material Random=true
   )
   &PartitionParticles(
    Normalized Index In Partition Includes One=true
    Number Of Partitions=4
    Particles Per Partition:NiagaraInt32=$Emitter.ParticlesPerTrail
    Partition Mode:ENiagara_PartitionMode=::NewEnumerator1
    Write Partition Index to Ribbon Link Order:ENiagara_AttributeSamplingApplyOutput=::NewEnumerator0
    Write Partition to Ribbon ID=::NewEnumerator0
   )
   /SetVariables(Particles.RibbonWidth=&SelectFloatFromArray(Direct Array Index=$Particles.PartitionParticles.Partition, Float Selection Array=DataInterface() {propertyValues {FloatData [38,38,25,25]}})) {isSetParameters=true}
   /SetVariables(Particles.RibbonLength:NiagaraFloat=&SelectFloatFromArray(
     Direct Array Index=$Particles.PartitionParticles.Partition
     Float Selection Array=DataInterface() {propertyValues {FloatData [150,150,100,100]}}
    )) {isSetParameters=true}
  ]
  particleUpdateScript [
   &ParticleState(Kill Particles When Lifetime Has Elapsed=false)
   &SocketLocation(
    Orientation Sampling=::NewEnumerator1
    Position Sampling=::NewEnumerator1
    SamplePosOffset=true
    Sampled Position Offset=&Multiply_VectorByFloat(
     Vector=&SelectVectorFromArray(
      Direct Array Index=$Particles.PartitionParticles.Partition
      Vector Selection Array=DataInterface() {propertyValues {FloatData [<0 0 -15>, <0 0 -15>, <0 0 -10>, <0 0 -10>]}}
     )
     Float=&FN_FrontEndFloatModifier(
      Gameplay=&Lerp_Float(A=1, B=1.3, Alpha=$System.ContrailDiveAlpha)
      Front End=&Lerp_Float(A=0.7, B=1, Alpha=$System.ContrailDiveAlpha)
     )
    )
    Sampled Position Offset Coordinate Space:ENiagaraSkelMeshTransforms=::NewEnumerator2
    Socket Index=$Particles.PartitionParticles.Partition
    Socket Reader=$System.Socket Reader
    Socket Sampling Mode=::NewEnumerator3
    VelocitySpeedLimit=false
   )
   /SetVariables(Particles.RibbonLength:NiagaraFloat=&Multiply_Float(
     A=$Particles.Initial.RibbonLength
     B=&FloatFromCurve(
      FloatCurve=DataInterface() {propertyValues {Curve {keys [
          {arriveTangent=0.1232, interpMode=RCIM_Cubic, leaveTangent=0.1232, tangentMode=RCTM_User, time=0, value=0}
          {interpMode=RCIM_Cubic, tangentMode=RCTM_User, time=1, value=1}
         ]}}}
      Scale Curve=&Lerp_Float(A=1, B=1.6, Alpha=$System.ContrailDiveAlpha)
     )
    )) {isSetParameters=true}
   /SetVariables(Particles.InterpolatedTrailVector:Vector3f=$Emitter.InterpolatedVector) {isSetParameters=true}
   /SetVariables(Particles.InterpolatedPosition:NiagaraPosition=&AddVectorToPosition(
     Position=$Output.SocketLocation.SampledPosition
     Vector=&Multiply_VectorByFloat(
      Vector=$Particles.InterpolatedTrailVector
      Float=&Multiply_Float(A=$Particles.RibbonLength, B=$Particles.PartitionParticles.Normalized Index In Partition)
     )
    )) {isSetParameters=true}
   /SetVariables(Particles.Position=$Particles.InterpolatedPosition) {isSetParameters=true}
   &ScaleRibbonWidth(Ribbon Width Scale=&Multiply_Float(
     A=&Lerp_Float(A=1, B=0.8, Alpha=$System.ContrailDiveAlpha)
     B=&Multiply_Float(
      A=&FloatFromCurve(
       FloatCurve=DataInterface() {propertyValues {Curve {keys [{interpMode=RCIM_Cubic, tangentMode=RCTM_User, time=0, value=0}, {interpMode=RCIM_Cubic, tangentMode=RCTM_User, time=1, value=1}]}}}
       Scale Curve=1.3
      )
      B=&MergeHelper_FloatSwitch(
       0=&FloatFromCurve(
        FloatCurve=DataInterface() {propertyValues {Curve {keys [
            {arriveTangent=2.2468, interpMode=RCIM_Cubic, leaveTangent=2.2468, tangentMode=RCTM_User, time=0, value=0.8}
            {arriveTangent=0.0009, interpMode=RCIM_Cubic, leaveTangent=0.0009, tangentMode=RCTM_User, time=0.2, value=1}
            {arriveTangent=-0.5878, interpMode=RCIM_Cubic, leaveTangent=-0.5878, tangentMode=RCTM_User, time=0.95, value=0.8}
           ]}}}
        CurveIndex=$Particles.RibbonLinkOrder
       )
       1=&FloatFromCurve(
        FloatCurve=DataInterface() {propertyValues {Curve {keys [
            {arriveTangent=2.2468, interpMode=RCIM_Cubic, leaveTangent=2.2468, tangentMode=RCTM_User, time=0, value=0.8}
            {arriveTangent=0.0009, interpMode=RCIM_Cubic, leaveTangent=0.0009, tangentMode=RCTM_User, time=0.2, value=1}
            {arriveTangent=-0.5878, interpMode=RCIM_Cubic, leaveTangent=-0.5878, tangentMode=RCTM_User, time=0.95, value=0.8}
           ]}}}
        CurveIndex=$Particles.RibbonLinkOrder
       )
       2=&FloatFromCurve(
        FloatCurve=DataInterface() {propertyValues {Curve {keys [
            {interpMode=RCIM_Cubic, tangentMode=RCTM_User, time=0.5, value=1}
            {arriveTangent=-2.9883, interpMode=RCIM_Cubic, leaveTangent=-2.9883, tangentMode=RCTM_User, time=1, value=0.4}
           ]}}}
        CurveIndex=$Particles.RibbonLinkOrder
       )
       3=&FloatFromCurve(
        FloatCurve=DataInterface() {propertyValues {Curve {keys [
            {interpMode=RCIM_Cubic, tangentMode=RCTM_User, time=0.5, value=1}
            {arriveTangent=-2.9883, interpMode=RCIM_Cubic, leaveTangent=-2.9883, tangentMode=RCTM_User, time=1, value=0.4}
           ]}}}
        CurveIndex=$Particles.RibbonLinkOrder
       )
       Index=$Particles.PartitionParticles.Partition
       Num=::NewEnumerator2
      )
     )
    ))
   /SetVariables(Particles.ToDTrailScale:LinearColor=&FNTimeOfDayColor(Night Color=<rgba 0.42 0.7467 1.4 1>, Morning Color=<rgba 1 0.7 0.7 1>, Day Color=<rgba 1 1 1 1>, Evening Color=<rgba 1.3 0.91 1.105 1>)) {isSetParameters=true}
   &Color(Color=&MultiplyLinearColor(
     Linear Color=&ColorFromCurve(
      ColorCurve=DataInterface() {propertyValues {
        AlphaCurve {keys [<at 0.7861 2.2 RCIM_Cubic>]}
        BlueCurve {keys [<at 0.0415 0.078>,<at 0.3038 0.042>,<at 0.9464 0.011>]}
        GreenCurve {keys [<at 0.0415 0.0892>,<at 0.3038 0.0476>,<at 0.9464 0.0124>]}
        RedCurve {keys [<at 0.0415 0.12>,<at 0.3038 0.07>,<at 0.9464 0.02>]}
       }}
      CurveIndex=$Particles.RibbonLinkOrder
     )
     Scale Factor=&MakeVector4_FromLinearColor(LinearColor=$Particles.ToDTrailScale)
    ))
   /SetVariables(Particles.Offset X:NiagaraFloat=&Add_Float(A=$Particles.Offset X, B=&Lerp_Float(A=0, B=-0.03, Alpha=$System.ContrailDiveAlpha))) {isSetParameters=true}
   &DynamicMaterialParameters(
    Index 0 Param 1=&SelectFloatFromArray(
     Direct Array Index=$Particles.PartitionParticles.Partition
     Float Selection Array=DataInterface() {propertyValues {FloatData [0.23,0.49,0.67,0.88]}}
    )
    Index 0 Param 2=$Particles.Offset X
    Index 0 Param 3=&Multiply_Float(
     A=&FloatFromCurve(
      FloatCurve=DataInterface() {propertyValues {Curve {keys [
          {arriveTangent=0.0484, interpMode=RCIM_Cubic, leaveTangent=0.0484, tangentMode=RCTM_User, time=0.3, value=1}
          {arriveTangent=0.0924, interpMode=RCIM_Cubic, leaveTangent=0.0924, tangentMode=RCTM_User, time=0.5, value=1}
          <at 1 0 RCIM_Cubic>
         ]}}}
      CurveIndex=$Particles.PartitionParticles.Normalized Index In Partition
     )
     B=&SelectFloatFromArray(Direct Array Index=$Particles.PartitionParticles.Partition, Float Selection Array=DataInterface() {propertyValues {FloatData [2,1.3,0,0]}})
    )
    Module.Index0Param0Write=true
    Module.Index0Param1Write=true
    Module.Index0Param2Write=true
    Module.Index0Param3Write=true
   ) {enabled=false}
   &DynamicMaterialParameters(
    Index 0 Param 1=&SelectFloatFromArray(
     Direct Array Index=$Particles.PartitionParticles.Partition
     Float Selection Array=DataInterface() {propertyValues {FloatData [0.3,0.9,0.1,0.7]}}
    )
    Index 0 Param 2=&FN_FrontEndFloatModifier(
     Gameplay=&Lerp_Float(A=1, B=20, Alpha=$System.ContrailDiveAlpha)
     Front End=&Lerp_Float(A=1, B=6, Alpha=$System.ContrailDiveAlpha)
    )
    Module.Index0Float0Write=true
    Module.Index0Float1Write=true
    Module.Index0Float2Write=false
    Module.Index0Float3Write=false
   ) {module=DynamicMaterialParameters001}
   /SetVariables(
    Particles.ColorCore:LinearColor=&MultiplyLinearColor(
     Linear Color=&ColorFromCurve(
      ColorCurve=DataInterface() {propertyValues {
        AlphaCurve {keys [<at 0.4918 2 RCIM_Cubic>,<at 1 0 RCIM_Cubic>]}
        BlueCurve {keys [<at 0.1785 0.075 RCIM_Cubic>]}
        GreenCurve {keys [<at 0.1785 0.1 RCIM_Cubic>]}
        RedCurve {keys [<at 0.1785 0.15 RCIM_Cubic>]}
       }}
      CurveIndex=$Particles.RibbonLinkOrder
     )
     Scale Factor=&MakeVector4_FromLinearColor(LinearColor=$Particles.ToDTrailScale)
    )
    Particles.RibbonLengthCore:NiagaraFloat=&Multiply_Float(A=$Particles.RibbonLength, B=0.75)
    Particles.RibbonWidthCore:NiagaraFloat=&Multiply_Float(A=$Particles.RibbonWidth, B=0.66)
   ) {isSetParameters=true}
  ]
  rendererClasses [&NiagaraRibbonRendererProperties]
  renderers [
   {rendererClass=&NiagaraRibbonRendererProperties}
   {rendererClass=&NiagaraRibbonRendererProperties, rendererIndex=1}
  ]
 }
 FootprintLeft {
  emitterSpawnScript [/SetVariables(Emitter.StepPosition:Vector3f=<40 -80 -25>) {isSetParameters=true, module=SetVariables_#12}]
  emitterUpdateScript [
   &EmitterState(
    Delay First Loop Only=true
    Life Cycle Mode=::NewEnumerator1
    Loop Delay=$System.FootstepsInitialDelay
    Loop Duration=$System.LoopDurationStep
    Recalculate Duration Each Loop=true
    UseLoopDelay=true
   )
   &SpawnBurst_Instantaneous(Spawn Probability=$System.FirstFootstepFadeIn, Use Spawn Probability=false, UseLoopCountLimit=false)
   &SpawnBurst_Instantaneous(
    Spawn Count=&MakeIntFromFloat(Float=&Multiply_Float(
      A=&MakeFloatFromInt(Int=&RandomRangeInt(Minimum=28, Maximum=40))
      B=$System.ScalabilityChunks
     ))
    Spawn Group=1
    Spawn Probability=$System.FirstFootstepFadeIn
    Spawn Time=0.32
    Use Spawn Probability=false
    UseLoopCountLimit=false
   ) {module=SpawnBurst_Instantaneous001}
   &SpawnBurst_Instantaneous(Spawn Group=2, Spawn Probability=$System.FirstFootstepFadeIn, Spawn Time=0.28, Use Spawn Probability=false, UseLoopCountLimit=false) {module=SpawnBurst_Instantaneous002}
   &SpawnBurst_Instantaneous(
    Spawn Count=3
    Spawn Group=3
    Spawn Probability=$System.FirstFootstepFadeIn
    Spawn Time=0.31
    Use Spawn Probability=false
    UseLoopCountLimit=false
   ) {module=SpawnBurst_Instantaneous003}
   &SpawnBurst_Instantaneous(
    Spawn Count=&MakeIntFromFloat(Float=&Multiply_Float(
      A=&MakeFloatFromInt(Int=&RandomRangeInt(Minimum=10, Maximum=20))
      B=$System.ScalabilitySparks
     ))
    Spawn Group=4
    Spawn Probability=$System.FirstFootstepFadeIn
    Spawn Time=0.3
    Use Spawn Probability=false
    UseLoopCountLimit=false
   ) {module=SpawnBurst_Instantaneous004}
  ]
  particleSpawnScript [
   &PartitionParticles(Number Of Partitions=&MergeHelper_IntSwitch(0=2, 1=47, 2=0, 3=3, 4=0, 5=?, Index=&SpawnGroup(), Num=::NewEnumerator3))
   &InitializeParticle(
    Lifetime=&MergeHelper_FloatSwitch(
     0=0.95
     1=&RandomRangeFloat(Minimum=0.2, Maximum=0.6)
     2=0.18
     3=&RandomRangeFloat(Minimum=0.13, Maximum=0.16)
     4=&RandomRangeFloat(Minimum=0.3, Maximum=0.5)
     Index=&SpawnGroup()
     Num=::NewEnumerator3
    )
    Mass Mode=::NewEnumerator3
    Mesh Scale Mode=::NewEnumerator3
    Mesh Uniform Scale=40
    Sprite Size Mode=::NewEnumerator3
    Uniform Sprite Size=45
    UsePositionOffset=false
    Write Material Random=true
   )
   /SetVariables(Particles.VisibilityTag=&SpawnGroup()) {isSetParameters=true, module=SetVariables_#13}
   &SpriteFacingAndAlignment(
    Sprite Alignment=<1 1 0>
    Sprite Facing=&MakeVector(X=0, Y=-0.01, Z=1)
    WriteAlignment=false
    WriteFacing=true
   )
   /SetVariables(
    Particles.MeshIndexChunks:NiagaraInt32=&RandomRangeInt(Minimum=0, Maximum=2)
    Particles.MeshOrientationChunks:Quat4f=/Niagara/DynamicInputs/Rotation/RandomQuaternion()
    Particles.ScaleChunks:Vector3f=&Multiply_VectorByFloat(
     Vector=&RandomRangeVector(Minimum=<0.8 0.8 0.8>, Maximum=<1 1 1>)
     Float=&RandomRangeFloat(Minimum=0.03, Maximum=0.04)
    )
   ) {isSetParameters=true, module=SetVariables_#14}
   &SocketLocation(
    Orientation Sampling=::NewEnumerator1
    Position Sampling=::NewEnumerator1
    SamplePosOffset=false
    Socket Index=4
    Socket Reader=$System.Socket Reader
    Socket Sampling Mode=::NewEnumerator3
    VelocitySpeedLimit=false
   ) {module=SocketLocation001}
   &ShapeLocation(
    MaskSpawnGroup=false
    Non Uniform Scale=<4 4 1>
    Offset=&RotateVectorByQuaternion(
     Vector To Rotate=&AddVectorToPosition(Position=Linked() {linkedVariable {name=Emitter.StepPosition, type=Vector3f}}, Vector=<13 -10 0>)
     Quaternion=$Output.SocketLocation001.SampledRotation
    )
    Offset Mode=::NewEnumerator0
    Shape Origin=$Output.SocketLocation001.SampledPosition
    Sphere Distribution=::NewEnumerator3
    Sphere Radius=3
    UseOwnerScale=true
   ) {module=ShapeLocation001}
   /SetVariables(Particles.Position=&MaskVectorBySpawnGroup(
     Masked Vector=&AddVectorToPosition(
      Position=$Output.SocketLocation001.SampledPosition
      Vector=&RotateVectorByQuaternion(
       Vector To Rotate=&AddVectorToPosition(
        Position=Linked() {linkedVariable {name=Emitter.StepPosition, type=Vector3f}}
        Vector=&SelectVectorFromArray(
         Direct Array Index=$Particles.PartitionParticles.Partition
         Vector Selection Array=DataInterface() {propertyValues {FloatData [<20 -7 11>, <24 -7 0>, <20 -7 -11>]}}
        )
       )
       Quaternion=$Output.SocketLocation001.SampledRotation
      )
     )
     Passthrough Vector=Linked() {linkedVariable {name=Particles.Position, type=NiagaraPosition}}
     Spawn Group Mask=3
    )) {isSetParameters=true}
   &AddVelocity(
    Module.Use Falloff Distance=false
    Origin Offset=&RotateVectorByQuaternion(
     Vector To Rotate=&AddVectorToPosition(Position=Linked() {linkedVariable {name=Emitter.StepPosition, type=Vector3f}}, Vector=<-5 10 0>)
     Quaternion=$Output.SocketLocation001.SampledRotation
    )
    Override Seed=false
    Velocity Mode=::NewEnumerator1
    Velocity Origin=$Output.SocketLocation001.SampledPosition
    Velocity Speed=&Multiply_Float(
     A=&RandomRangeFloat(Minimum=0.5, Maximum=1)
     B=&MaskFloat(
      Mask Parameter=$Emitter.SpawnGroup
      Mask=4
      Masked Float=&RandomRangeFloat(Minimum=30, Maximum=100)
      Passthrough Float=220
     )
    )
   )
   &CurlNoiseForce(
    Bias Noise=true
    Bias Noise Field:Vector3f=&RotateVectorByQuaternion(
     Vector To Rotate=&RandomRangeVector(Minimum=<0.1 -0.1 0>, Maximum=<0.4 -0.2 0>)
     Quaternion=$Output.SocketLocation001.SampledRotation
    )
    Noise Frequency=300
    Noise Strength=&MaskFloat(
     Mask Parameter=$Emitter.SpawnGroup
     Mask=4
     Masked Float=&RandomRangeFloat(Minimum=40, Maximum=100)
     Passthrough Float=0
    )
    Pan Noise=true
    Pan Noise Field=<0 0.1 0>
   )
   &InitialMeshOrientation(Mesh Orientation Mode=::NewEnumerator2, Orientation Vector=<0 0 1>, Rotation=<0.25 0 0>, Use Rotation Vector=true)
   &ApplyInitialForces()
  ]
  particleUpdateScript [
   &SocketLocation(
    Orientation Sampling=::NewEnumerator1
    Position Sampling=::NewEnumerator1
    SamplePosOffset=false
    Socket Index=4
    Socket Reader=$System.Socket Reader
    Socket Sampling Mode=::NewEnumerator3
    VelocitySpeedLimit=false
   )
   /SetVariables(Particles.FXPositionOffset:NiagaraPosition=&AddVectorToPosition(
     Position=$Output.SocketLocation.SampledPosition
     Vector=&RotateVectorByQuaternion(Vector To Rotate=$Emitter.StepPosition, Quaternion=$Output.SocketLocation.SampledRotation)
    )) {isSetParameters=true, module=SetVariables_#15}
   &ParticleState()
   &ScaleSpriteSize(Uniform Curve Sprite Scale=DataInterface() {propertyValues {Curve {keys [
        {arriveTangent=8.5804, interpMode=RCIM_Cubic, leaveTangent=8.5804, tangentMode=RCTM_User, time=0, value=0.1}
        {interpMode=RCIM_Cubic, tangentMode=RCTM_User, time=0.2, value=1}
       ]}}})
   &ScaleColor(
    Scale Alpha=&FloatFromCurve(FloatCurve=DataInterface() {propertyValues {Curve {keys [
         {arriveTangent=4.5659, interpMode=RCIM_Cubic, leaveTangent=4.5659, tangentMode=RCTM_User, time=0, value=0}
         {arriveTangent=0.0464, leaveTangent=0.0464, time=0.1, value=1}
         {arriveTangent=-1.712, interpMode=RCIM_Cubic, leaveTangent=-2.3018, tangentMode=RCTM_Break, time=0.35, value=1}
         {arriveTangent=-0.1251, interpMode=RCIM_Cubic, leaveTangent=-0.1251, tangentMode=RCTM_User, time=1, value=0}
        ]}}})
    Scale RGB=&MakeVectorFromLinearColorRGB(Color=&ColorFromCurve(ColorCurve=DataInterface() {propertyValues {
        AlphaCurve {keys [<at 0 1 RCIM_Cubic>]}
        BlueCurve {keys [<at 0.474 0.7 RCIM_Cubic>,<at 0.6955 0.1 RCIM_Cubic>]}
        GreenCurve {keys [<at 0.474 0.7 RCIM_Cubic>,<at 0.6955 0.1 RCIM_Cubic>]}
        RedCurve {keys [<at 0.474 0.7 RCIM_Cubic>,<at 0.6955 0.1 RCIM_Cubic>]}
       }}))
    ScaleA=true
    ScaleRGB=true
   )
   &Drag(Drag=&RandomRangeFloat(Minimum=1, Maximum=5), Use Linear Drag=true, Use Rotational Drag=true)
   &SolveForcesAndVelocity(Clamp Velocity=false, Limit Acceleration=false)
   /SetVariables(
    Particles.DynParFootprint:Vector4f=&MakeVector4(
     X=&FloatFromCurve(FloatCurve=DataInterface() {propertyValues {Curve {keys [
          {arriveTangent=-8.1467, interpMode=RCIM_Cubic, leaveTangent=-8.1467, tangentMode=RCTM_User, time=0, value=1}
          {arriveTangent=-0.8705, interpMode=RCIM_Cubic, leaveTangent=-0.8705, tangentMode=RCTM_User, time=0.15, value=0.1}
          {arriveTangent=0.136, interpMode=RCIM_Cubic, leaveTangent=0.136, tangentMode=RCTM_User, time=0.2, value=0}
          {arriveTangent=0.2628, interpMode=RCIM_Cubic, leaveTangent=0.2628, tangentMode=RCTM_User, time=0.33, value=0}
          {arriveTangent=2.9851, interpMode=RCIM_Cubic, leaveTangent=2.9851, time=0.5, value=0.25}
          {arriveTangent=-0.0282, interpMode=RCIM_Cubic, leaveTangent=-0.0282, tangentMode=RCTM_User, time=1, value=2}
         ]}}})
     Y=&FloatFromCurve(
      FloatCurve=DataInterface() {propertyValues {Curve {keys [
          {arriveTangent=0.749, interpMode=RCIM_Cubic, leaveTangent=0.749, tangentMode=RCTM_User, time=0.1, value=0.1}
          {arriveTangent=6.5311, interpMode=RCIM_Cubic, leaveTangent=6.5311, tangentMode=RCTM_User, time=0.2, value=0.3}
          {arriveTangent=-0.4547, interpMode=RCIM_Cubic, leaveTangent=-0.4547, tangentMode=RCTM_User, time=0.33, value=1}
          {arriveTangent=-3.7037, interpMode=RCIM_Cubic, leaveTangent=-3.7037, time=0.4, value=0.5}
          {arriveTangent=-0.7364, interpMode=RCIM_Cubic, leaveTangent=-0.7364, tangentMode=RCTM_User, time=0.6, value=0}
         ]}}}
      Scale Curve=2
     )
     Z=0
     W=0
    )
    Particles.ScaleFootprint:Vector3f=&Multiply_VectorByFloat(
     Vector=<0.2 0.2 0.26>
     Float=&FloatFromCurve(FloatCurve=DataInterface() {propertyValues {Curve {keys [{interpMode=RCIM_Cubic, tangentMode=RCTM_User, time=0, value=0}, {interpMode=RCIM_Cubic, tangentMode=RCTM_User, time=0.1, value=1}]}}})
    )
   ) {isSetParameters=true, module=SetVariables_#16}
   /SetVariables(Particles.PositionFootprint:NiagaraPosition=&AddVectorToPosition(
     Position=$Particles.FXPositionOffset
     Vector=&RotateVectorByQuaternion(
      Vector To Rotate=&MakeVector(
       X=&Lerp_Float(
        A=&Lerp_Float(A=-110, B=-130, Alpha=$System.DiveSprint)
        B=1
        Alpha=&FloatFromCurve(FloatCurve=DataInterface() {propertyValues {Curve {keys [
             {arriveTangent=6.5872, interpMode=RCIM_Cubic, leaveTangent=6.5872, tangentMode=RCTM_User, time=0, value=0}
             {arriveTangent=-0.9166, interpMode=RCIM_Cubic, leaveTangent=-0.9166, tangentMode=RCTM_User, time=0.4, value=1}
             {arriveTangent=-0.4877, interpMode=RCIM_Cubic, leaveTangent=-0.4877, tangentMode=RCTM_User, time=0.8, value=0.2}
            ]}}})
       )
       Y=&Lerp_Float(
        A=&Lerp_Float(
         A=-50
         B=-5
         Alpha=&FloatFromCurve(FloatCurve=DataInterface() {propertyValues {Curve {keys [
              <at 0 0.5 RCIM_Cubic>
              {arriveTangent=1.4787, interpMode=RCIM_Cubic, leaveTangent=0.025, tangentMode=RCTM_Break, time=0.13, value=0}
              {
               arriveTangent=11.0317
               arriveTangentWeight=0.4523
               interpMode=RCIM_Cubic
               leaveTangent=2.8039
               leaveTangentWeight=0.1357
               tangentMode=RCTM_Break
               time=0.3
               value=0.9
              }
              {arriveTangent=0.0143, interpMode=RCIM_Cubic, leaveTangent=0.0143, tangentMode=RCTM_Break, time=0.36, value=1}
              {arriveTangent=-3.3431, interpMode=RCIM_Cubic, leaveTangent=-3.3431, tangentMode=RCTM_User, time=0.6, value=0.7}
              {arriveTangent=0.0231, interpMode=RCIM_Cubic, leaveTangent=0.0231, tangentMode=RCTM_User, time=0.85, value=0.4}
             ]}}})
        )
        B=&Lerp_Float(
         A=-50
         B=-5
         Alpha=&FloatFromCurve(FloatCurve=DataInterface() {propertyValues {Curve {keys [
              <at 0 0.5 RCIM_Cubic>
              {arriveTangent=1.4787, interpMode=RCIM_Cubic, leaveTangent=0.025, tangentMode=RCTM_Break, time=0.13, value=0}
              {
               arriveTangent=11.0317
               arriveTangentWeight=0.4523
               interpMode=RCIM_Cubic
               leaveTangent=2.8039
               leaveTangentWeight=0.1357
               tangentMode=RCTM_Break
               time=0.3
               value=0.9
              }
              {arriveTangent=0.0143, interpMode=RCIM_Cubic, leaveTangent=0.0143, tangentMode=RCTM_Break, time=0.36, value=1}
              {arriveTangent=-1.7647, interpMode=RCIM_Cubic, leaveTangent=-1.7647, time=0.5, value=0.875}
              {arriveTangent=-4.2087, interpMode=RCIM_Cubic, leaveTangent=-4.2087, tangentMode=RCTM_User, time=0.7, value=0.4}
             ]}}})
        )
        Alpha=$System.DiveSprint
       )
       Z=0
      )
      Quaternion=$Output.SocketLocation.SampledRotation
     )
    )) {isSetParameters=true, module=SetVariables_#17}
   &UpdateMeshOrientation(
    Facing Direction=&RotateVectorByQuaternion(Vector To Rotate=<1 0 0>, Quaternion=$Output.SocketLocation.SampledRotation)
    Facing Interpolation=false
    Orientation Method=::NewEnumerator1
    Override Quat=false
    TransformFacing=false
    TransformUp=false
    Up Direction:Vector3f=<-1 0 0>
   ) {module=UpdateMeshOrientation001}
   &UpdateMeshOrientation(
    Facing Interpolation=false
    Orientation Method=::NewEnumerator1
    Override Quat=false
    TransformFacing=false
    TransformUp=false
    Up Direction:Vector3f=&MakeVector(
     X=0
     Y=&Lerp_Float(
      A=1
      B=-1
      Alpha=&FloatFromCurve(FloatCurve=DataInterface() {propertyValues {Curve {keys [
           {interpMode=RCIM_Cubic, tangentMode=RCTM_User, time=0.1, value=0}
           {arriveTangent=9.0687, interpMode=RCIM_Cubic, leaveTangent=-0.2697, tangentMode=RCTM_Break, time=0.25, value=1}
           {arriveTangent=-1.7789, interpMode=RCIM_Cubic, leaveTangent=-1.7789, tangentMode=RCTM_User, time=0.33, value=0.6}
           {arriveTangent=0.0864, interpMode=RCIM_Cubic, leaveTangent=0.0864, tangentMode=RCTM_User, time=0.5, value=0.4}
           <at 1 0.7 RCIM_Cubic>
          ]}}})
     )
     Z=&FloatFromCurve(FloatCurve=DataInterface() {propertyValues {Curve {keys [
          {interpMode=RCIM_Cubic, tangentMode=RCTM_User, time=0, value=0.25}
          {interpMode=RCIM_Cubic, tangentMode=RCTM_User, time=0.3, value=1}
          {arriveTangent=-0.5714, interpMode=RCIM_Cubic, leaveTangent=-0.5714, time=0.5, value=0.5}
          {arriveTangent=-0.397, interpMode=RCIM_Cubic, leaveTangent=-0.397, tangentMode=RCTM_User, time=1, value=0.6}
         ]}}})
    )
   ) {module=UpdateMeshOrientation002}
   /SetVariables(Particles.MeshOrientationImpact:Quat4f=&MultiplyQuaternion(
     Quaternion A=&MultiplyQuaternion(Quaternion A=$Output.SocketLocation.SampledRotation, Quaternion B=<1 0 0 1>)
     Quaternion B=<0 0 1 1>
    )) {isSetParameters=true, module=SetVariables_#18}
   /SetVariables(
    Particles.ColorChunks:LinearColor=&ColorFromCurve(ColorCurve=DataInterface() {propertyValues {
       AlphaCurve {keys [<at 0 1 RCIM_Cubic>]}
       BlueCurve {keys [<at 0 1 RCIM_Cubic>,<at 0.1442 0.1812 RCIM_Cubic>]}
       GreenCurve {keys [<at 0 20.8 RCIM_Cubic>,<at 0.1442 0.5693 RCIM_Cubic>]}
       RedCurve {keys [<at 0 100 RCIM_Cubic>,<at 0.1442 1 RCIM_Cubic>]}
      }})
    Particles.PositionRockChunks:NiagaraPosition=&AddVectorToPosition(
     Position=$Particles.FXPositionOffset
     Vector=&AddVectorToPosition(
      Position=&MakeVector(
       X=0
       Y=&Lerp_Float(
        A=0
        B=&Lerp_Float(A=-60, B=-25, Alpha=$System.DiveSprint)
        Alpha=&FloatFromCurve(
         FloatCurve=DataInterface() {propertyValues {Curve {keys [
             {arriveTangent=0.6268, interpMode=RCIM_Cubic, leaveTangent=0.6268, tangentMode=RCTM_User, time=0, value=0}
             {arriveTangent=0.8137, interpMode=RCIM_Cubic, leaveTangent=0.8137, tangentMode=RCTM_User, time=1, value=1}
            ]}}}
         Scale Curve=&RandomRangeFloat(Minimum=0.4, Maximum=1.5)
        )
       )
       Z=&Lerp_Float(
        A=0
        B=&Lerp_Float(A=80, B=230, Alpha=$System.DiveSprint)
        Alpha=&FloatFromCurve(
         FloatCurve=DataInterface() {propertyValues {Curve {keys [
             {arriveTangent=0.3872, interpMode=RCIM_Cubic, leaveTangent=0.3872, tangentMode=RCTM_User, time=0.2, value=0}
             {arriveTangent=1.8857, interpMode=RCIM_Cubic, leaveTangent=1.8857, tangentMode=RCTM_User, time=1, value=1}
            ]}}}
         Scale Curve=&Multiply_Float(
          A=&RandomRangeFloat(Minimum=0.75, Maximum=1)
          B=&FN_FrontEndFloatModifier(Gameplay=1, Front End=0.5)
         )
        )
       )
      )
      Vector=&Multiply_VectorByFloat(
       Vector=&RotateVectorByQuaternion(
        Vector To Rotate=&SelectVectorFromArray(
         Direct Array Index=$Particles.PartitionParticles.Partition
         Vector Selection Array=DataInterface() {propertyValues {FloatData [
            <-0.2995 0.4197 0.0001>
            <-0.3102 0.3246 0.0001>
            <-0.306 0.3354 0.0003>
            <-0.3246 0.1977 0.0003>
            <-0.4562 -0.0267 0.0002>
            <-0.5077 -0.3741 0.0001>
            <-0.4908 -0.4304 0.0003>
            <-0.5093 -0.3564 0.0003>
            <-0.4796 -0.4587 0.0001>
            <-0.4208 -0.4669 0.0001>
            <-0.2903 -0.2606 0.0003>
            <-0.1981 -0.0135 0.0001>
            <-0.1088 0.115 0.0003>
            <-0.099 -0.0415 0.0003>
            <-0.1002 -0.0748 0.0002>
            <-0.1166 -0.2389 0.0002>
            <-0.1085 -0.1962 0.0001>
            <-0.1078 -0.2028 0>
            <-0.1195 -0.3133 0.0002>
            <-0.1064 -0.4026 0.0001>
            <-0.0947 -0.5726 0.0002>
            <-0.0707 -0.6709 0.0003>
            <-0.0505 -0.7363 0.0001>
            <-0.0904 -0.5669 0.0001>
            <-0.0317 -0.8066 0.0003>
            <0.1074 -0.5969 0.0003>
            <0.1375 -0.4617 0.0002>
            <0.1383 -0.4211 0.0003>
            <0.1418 -0.1988 0.0001>
            <0.137 -0.007 0.0003>
            <0.1796 0.0867 0.0002>
            <0.2761 -0.1702 0>
            <0.4498 -0.512 0.0003>
            <0.4742 -0.4529 0.0002>
            <0.4855 -0.4142 0.0003>
            <0.5012 -0.1966 0.0003>
            <0.4866 -0.1118 0>
            <0.3754 0.143 0.0003>
            <0.3499 0.3494 0.0003>
            <0.2809 0.6532 0.0002>
            <0.2329 0.6902 0.0001>
            <0.2263 0.6914 0.0003>
            <0.2093 0.7066 0.0002>
            <0.0516 0.8697 0>
            <-0.1207 0.7843 0.0003>
            <-0.1434 0.764 0.0002>
            <-0.1317 0.7705 0.0002>
            <-0.2533 0.6794 0.0002>
           ]}}
        )
        Quaternion=&MultiplyQuaternion(
         Quaternion A=$Output.SocketLocation.SampledRotation
         Quaternion B=&OrientMeshToVector(Look At Direction=<-1 0 0>, Reference Axis Vector=<1 0 0>)
        )
       )
       Float=&Multiply_Float(
        A=-25
        B=&FloatFromCurve(FloatCurve=DataInterface() {propertyValues {Curve {keys [
             {arriveTangent=1.6232, interpMode=RCIM_Cubic, leaveTangent=1.6232, tangentMode=RCTM_User, time=0.1, value=0.8}
             {interpMode=RCIM_Cubic, tangentMode=RCTM_User, time=0.5, value=1}
            ]}}})
       )
      )
     )
    )
   ) {isSetParameters=true, module=SetVariables_#19}
   /SetVariables(Particles.ScaleChunks:Vector3f=&Multiply_VectorByFloat(
     Vector=$Particles.Initial.ScaleChunks
     Float=&FloatFromCurve(FloatCurve=DataInterface() {propertyValues {Curve {keys [
          {interpMode=RCIM_Cubic, tangentMode=RCTM_User, time=0, value=0}
          {interpMode=RCIM_Cubic, tangentMode=RCTM_User, time=0.14, value=1}
          {arriveTangent=-1.9479, interpMode=RCIM_Cubic, leaveTangent=-1.9479, tangentMode=RCTM_User, time=1, value=0}
         ]}}})
    )) {isSetParameters=true, module=SetVariables_#20}
   /SetVariables(
    Particles.ColorFlames:LinearColor=&ColorFromCurve(ColorCurve=DataInterface() {propertyValues {
       AlphaCurve {keys [<at 0 0>,<at 0.0773 1 RCIM_Cubic>,<at 1 0 RCIM_Cubic>]}
       BlueCurve {keys [<at 0.0656 0.14>,<at 0.4845 0.03>,<at 0.9133 0 RCIM_Cubic>]}
       GreenCurve {keys [<at 0.0656 1.7407>,<at 0.4845 0.1593>,<at 0.9133 0.0083 RCIM_Cubic>]}
       RedCurve {keys [<at 0.0656 7>,<at 0.4845 1>,<at 0.9133 1 RCIM_Cubic>]}
      }})
    Particles.ColorImpact:LinearColor=&ColorFromCurve(ColorCurve=DataInterface() {propertyValues {
       AlphaCurve {keys [<at 0.0107 0>,<at 0.1556 1 RCIM_Cubic>,<at 1 0>]}
       BlueCurve {keys [<at 0.1512 1.7>,<at 1 0 RCIM_Cubic>]}
       GreenCurve {keys [<at 0.1512 1.7>,<at 1 0 RCIM_Cubic>]}
       RedCurve {keys [<at 0.1512 1.7>,<at 1 0 RCIM_Cubic>]}
      }})
    Particles.DynamicMatParFlames:Vector4f=&MakeVector4(
     X=&FloatFromCurve(FloatCurve=DataInterface() {propertyValues {Curve {keys [<at 0 1>,<at 1 0.7>]}}})
     Y=0
     Z=0
     W=0
    )
    Particles.PositionImpact:NiagaraPosition=$Particles.FXPositionOffset
    Particles.ScaleImpact:Vector3f=&Multiply_VectorByFloat(
     Vector=<0.9 1 1>
     Float=&FloatFromCurve(
      FloatCurve=DataInterface() {propertyValues {Curve {keys [
          {arriveTangent=0.897, interpMode=RCIM_Cubic, leaveTangent=0.897, tangentMode=RCTM_User, time=0, value=0.6}
          {interpMode=RCIM_Cubic, tangentMode=RCTM_User, time=0.9, value=1}
         ]}}}
      Scale Curve=0.33
     )
    )
    Particles.SpriteSizeFlames:Vector2f=&Multiply_Vector2D_ByFloat(
     Float=&Multiply_Float(
      A=&RandomRangeFloat(Minimum=37, Maximum=42)
      B=&MergeHelper_FloatSwitch(0=1, 1=1.2, 2=1, 3=?, Index=$Particles.PartitionParticles.Partition, Num=::NewEnumerator1)
     )
     Vector2D=&MakeVector2D(
      X=&FloatFromCurve(FloatCurve=DataInterface() {propertyValues {Curve {keys [<at 0.3 1>,<at 1 0.7>]}}}, Scale Curve=0.8)
      Y=&FloatFromCurve(
       FloatCurve=DataInterface() {propertyValues {Curve {keys [
           {arriveTangent=-0.5687, interpMode=RCIM_Cubic, leaveTangent=-0.5687, tangentMode=RCTM_User, time=0.5, value=1}
           {arriveTangent=-1.6572, interpMode=RCIM_Cubic, leaveTangent=-1.6572, tangentMode=RCTM_User, time=1, value=0.5}
          ]}}}
       Scale Curve=1.4
      )
     )
    )
   ) {isSetParameters=true, module=SetVariables_#21}
   /SetVariables(Particles.ColorImpactCore:LinearColor=&MultiplyLinearColor(
     Linear Color=$Particles.ColorImpact
     Scale Factor=&MakeFloat4FromVectorAndFloat(
      "Vector (RGB)"=&VectorFromFloat(Value=&Lerp_Float(
        A=50
        B=1
        Alpha=&FloatFromCurve(FloatCurve=DataInterface() {propertyValues {Curve {keys [{interpMode=RCIM_Cubic, tangentMode=RCTM_User, time=0.33, value=0}, {interpMode=RCIM_Cubic, tangentMode=RCTM_User, time=0.4, value=1}]}}})
       ))
      "Float (Alpha)"=1
     )
    )) {isSetParameters=true}
   /SetVariables(Particles.DynamicMatParFlameBursts:Vector4f=&MakeVector4(X=&RandomRangeFloat(Minimum=0, Maximum=1), Y=0, Z=0, W=0)) {isSetParameters=true}
   /SetVariables(
    Particles.ColorSparks:LinearColor=&ColorFromCurve(ColorCurve=DataInterface() {propertyValues {
       AlphaCurve {keys [<at 0 0>, {arriveTangent=1.1943, interpMode=RCIM_Cubic, leaveTangent=1.1943, time=0.0773, value=1}, <at 0.8373 1>, <at 1 0 RCIM_Cubic>]}
       BlueCurve {keys [<at 0.0985 0.5>,<at 0.2018 0.12>,<at 0.6092 0>,<at 0.9612 0>]}
       GreenCurve {keys [<at 0.0985 2.3>,<at 0.2018 0.59>,<at 0.6092 0.1>,<at 0.9612 0.0033>]}
       RedCurve {keys [<at 0.0985 5>,<at 0.2018 2>,<at 0.6092 1>,<at 0.9612 0.1>]}
      }})
    Particles.SpriteSizeSparks:Vector2f=&Multiply_Vector2D_ByFloat(
     Float=&RandomRangeFloat(Minimum=10, Maximum=15)
     Vector2D=&Multiply_Vector2D_ByFloat(
      Float=&FloatFromCurve(FloatCurve=DataInterface() {propertyValues {Curve {keys [
           {interpMode=RCIM_Cubic, tangentMode=RCTM_User, time=0, value=1}
           {arriveTangent=-2.4141, interpMode=RCIM_Cubic, leaveTangent=-2.4141, tangentMode=RCTM_User, time=1, value=0}
          ]}}})
      Vector2D=&MakeVector2D(X=0.7, Y=&RandomRangeFloat(Minimum=0.8, Maximum=1))
     )
    )
   ) {isSetParameters=true, module=SetVariables_#22}
   /SetVariables(
    Particles.ColorDirt:LinearColor=&ColorFromCurve(ColorCurve=DataInterface() {propertyValues {
       AlphaCurve {keys [<at 0.3044 0>,<at 0.3827 1 RCIM_Cubic>,<at 0.5082 1>,<at 0.84 0>]}
       BlueCurve {keys [<at 0 0.02 RCIM_Cubic>]}
       GreenCurve {keys [<at 0 0.025 RCIM_Cubic>]}
       RedCurve {keys [<at 0 0.04 RCIM_Cubic>]}
      }})
    Particles.PositionDirt:NiagaraPosition=&AddVectorToPosition(
     Position=$Particles.PositionImpact
     Vector=&RotateVectorByQuaternion(
      Vector To Rotate=&MakeVector(
       X=&Lerp_Float(
        A=0
        B=-20
        Alpha=&FloatFromCurve(FloatCurve=DataInterface() {propertyValues {Curve {keys [<at 0.4 0>,<at 0.9 1>]}}})
       )
       Y=&Lerp_Float(
        A=0
        B=-5
        Alpha=&FloatFromCurve(FloatCurve=DataInterface() {propertyValues {Curve {keys [<at 0.4 0>,<at 0.7 1>]}}})
       )
       Z=0
      )
      Quaternion=$Output.SocketLocation.SampledRotation
     )
    )
    Particles.SpriteSizeDirt:Vector2f=&Multiply_Vector2D_ByFloat(
     Float=&FloatFromCurve(FloatCurve=DataInterface() {propertyValues {Curve {keys [
          <at 0.3 0 RCIM_Cubic>
          {arriveTangent=3.8859, interpMode=RCIM_Cubic, leaveTangent=3.8859, tangentMode=RCTM_User, time=0.4, value=0.4}
          {interpMode=RCIM_Cubic, tangentMode=RCTM_User, time=0.7, value=1}
         ]}}})
     Vector2D=&Vector2DFromFloat(Value=&RandomRangeFloat(Minimum=70, Maximum=85))
    )
   ) {isSetParameters=true, module=SetVariables_#23}
   /SetVariables(
    Particles.ColorTrail:LinearColor=&MultiplyLinearColor(
     Linear Color=<rgba 1 1 1 1>
     Scale Factor=&MakeVector4(
      X=1
      Y=1
      Z=1
      W=&FloatFromCurve(
       FloatCurve=DataInterface() {propertyValues {Curve {keys [
           {interpMode=RCIM_Cubic, tangentMode=RCTM_User, time=0.05, value=0}
           {interpMode=RCIM_Cubic, tangentMode=RCTM_User, time=0.11, value=1}
           {interpMode=RCIM_Cubic, tangentMode=RCTM_User, time=0.14, value=1}
           {interpMode=RCIM_Cubic, tangentMode=RCTM_User, time=0.16, value=0}
           {interpMode=RCIM_Cubic, tangentMode=RCTM_User, time=0.2, value=0}
           {interpMode=RCIM_Cubic, tangentMode=RCTM_User, time=0.28, value=1.4}
           {interpMode=RCIM_Cubic, tangentMode=RCTM_User, time=0.33, value=1}
           {interpMode=RCIM_Cubic, tangentMode=RCTM_User, time=0.5, value=0}
          ]}}}
       Scale Curve=&FN_FrontEndFloatModifier(Gameplay=1.7, Front End=1)
      )
     )
    )
    Particles.DynamicMatParTrail:Vector4f=&MakeVector4(
     X=1
     Y=&FloatFromCurve(
      FloatCurve=DataInterface() {propertyValues {Curve {keys [
          {arriveTangent=-12.4698, interpMode=RCIM_Cubic, leaveTangent=-12.4698, tangentMode=RCTM_User, time=0.02, value=0.65}
          {interpMode=RCIM_Cubic, tangentMode=RCTM_User, time=0.15, value=0.1}
          {arriveTangent=-1.0237, interpMode=RCIM_Cubic, leaveTangent=-1.0237, tangentMode=RCTM_User, time=0.2, value=0.25}
          {arriveTangent=-0.4703, interpMode=RCIM_Cubic, leaveTangent=-0.4703, tangentMode=RCTM_User, time=0.5, value=0.05}
         ]}}}
      Scale Curve=&FN_FrontEndFloatModifier(Gameplay=-0.33, Front End=-0.3)
     )
     Z=0
     W=0
    )
    Particles.ScaleTrail:Vector3f=&Multiply_Vector(
     A=$Particles.ScaleFootprint
     B=&MakeVector(
      X=1
      Y=1
      Z=&FloatFromCurve(
       FloatCurve=DataInterface() {propertyValues {Curve {keys [
           {arriveTangent=6.6357, interpMode=RCIM_Cubic, leaveTangent=6.6357, tangentMode=RCTM_User, time=0, value=0.2}
           {arriveTangent=-2.8877, interpMode=RCIM_Cubic, leaveTangent=-2.8877, tangentMode=RCTM_User, time=0.1, value=1}
           {interpMode=RCIM_Cubic, tangentMode=RCTM_User, time=0.17, value=0}
           {arriveTangent=-0.0918, interpMode=RCIM_Cubic, leaveTangent=-0.0918, tangentMode=RCTM_User, time=0.28, value=1}
           {arriveTangent=-5.8824, interpMode=RCIM_Cubic, leaveTangent=-5.8824, time=0.37, value=0.2}
           {interpMode=RCIM_Cubic, tangentMode=RCTM_User, time=0.45, value=0}
          ]}}}
       Scale Curve=&Lerp_Float(A=2, B=2, Alpha=$System.DiveSprint)
      )
     )
    )
    Particles.TrailOrientation:Quat4f=&MultiplyQuaternion(
     Quaternion A=$Particles.MeshOrientation
     Quaternion B=/Niagara/DynamicInputs/TypeConversions/MakeCustomQuaternionFromBool(
      Bool:NiagaraBool=&SetBoolByFloatComparison(
       A=&FloatFromCurve(FloatCurve=DataInterface() {propertyValues {Curve {keys [<at 0.1 1>,<at 0.3 0>,<at 1 0>]}}})
       B=0.5
      )
      True Quat:Quat4f=&MultiplyQuaternion(Quaternion A=<0 1 0 1>, Quaternion B=<0 1 0 1>)
      False Quat:Quat4f=<0 0 0 1>
     )
    )
   ) {isSetParameters=true}
  ]
  rendererClasses [&NiagaraMeshRendererProperties, &NiagaraSpriteRendererProperties]
  renderers [
   {rendererClass=&NiagaraMeshRendererProperties, rendererIndex=0}
   {rendererClass=&NiagaraMeshRendererProperties, rendererIndex=1}
   {rendererClass=&NiagaraSpriteRendererProperties, rendererIndex=2}
   {rendererClass=&NiagaraSpriteRendererProperties, rendererIndex=3}
   {rendererClass=&NiagaraMeshRendererProperties, rendererIndex=4}
   {rendererClass=&NiagaraMeshRendererProperties, rendererIndex=5}
   {rendererClass=&NiagaraSpriteRendererProperties, rendererIndex=6}
   {rendererClass=&NiagaraMeshRendererProperties, rendererIndex=7}
  ]
 }
 FootprintRight {
  emitterSpawnScript [/SetVariables(Emitter.StepPosition:Vector3f=<40 -80 25>) {isSetParameters=true, module=SetVariables_#12}]
  emitterUpdateScript [
   &EmitterState(
    Delay First Loop Only=true
    Life Cycle Mode=::NewEnumerator1
    Loop Delay=$System.FootstepsInitialDelay
    Loop Duration=$System.LoopDurationStep
    Recalculate Duration Each Loop=true
    UseLoopDelay=true
   )
   /SetVariables(Emitter.CustomAge:NiagaraFloat=&Add_Float(A=$Emitter.LoopedAge, B=&Multiply_Float(A=$Emitter.CurrentLoopDuration, B=-0.5))) {isSetParameters=true}
   &SpawnBurst_Instantaneous(Age:NiagaraFloat=$Emitter.CustomAge, Use Spawn Probability=false, UseLoopCountLimit=false)
   &SpawnBurst_Instantaneous(
    Age:NiagaraFloat=$Emitter.CustomAge
    Spawn Count=&MakeIntFromFloat(Float=&Multiply_Float(
      A=&MakeFloatFromInt(Int=&RandomRangeInt(Minimum=28, Maximum=40))
      B=$System.ScalabilityChunks
     ))
    Spawn Group=1
    Spawn Time=0.32
    Use Spawn Probability=false
    UseLoopCountLimit=false
   ) {module=SpawnBurst_Instantaneous001}
   &SpawnBurst_Instantaneous(Age:NiagaraFloat=$Emitter.CustomAge, Spawn Group=2, Spawn Time=0.28, Use Spawn Probability=false, UseLoopCountLimit=false) {module=SpawnBurst_Instantaneous002}
   &SpawnBurst_Instantaneous(Age:NiagaraFloat=$Emitter.CustomAge, Spawn Count=3, Spawn Group=3, Spawn Time=0.31, Use Spawn Probability=false, UseLoopCountLimit=false) {module=SpawnBurst_Instantaneous003}
   &SpawnBurst_Instantaneous(
    Age:NiagaraFloat=$Emitter.CustomAge
    Spawn Count=&MakeIntFromFloat(Float=&Multiply_Float(
      A=&MakeFloatFromInt(Int=&RandomRangeInt(Minimum=10, Maximum=20))
      B=$System.ScalabilitySparks
     ))
    Spawn Group=4
    Spawn Time=0.3
    Use Spawn Probability=false
    UseLoopCountLimit=false
   ) {module=SpawnBurst_Instantaneous004}
  ]
  particleSpawnScript [
   &PartitionParticles(Number Of Partitions=&MergeHelper_IntSwitch(0=2, 1=47, 2=0, 3=3, 4=0, 5=?, Index=&SpawnGroup(), Num=::NewEnumerator3))
   &InitializeParticle(
    Lifetime=&MergeHelper_FloatSwitch(
     0=0.95
     1=&RandomRangeFloat(Minimum=0.2, Maximum=0.6)
     2=0.18
     3=&RandomRangeFloat(Minimum=0.13, Maximum=0.16)
     4=&RandomRangeFloat(Minimum=0.3, Maximum=0.5)
     Index=&SpawnGroup()
     Num=::NewEnumerator3
    )
    Mass Mode=::NewEnumerator3
    Mesh Scale Mode=::NewEnumerator3
    Mesh Uniform Scale=40
    Sprite Size Mode=::NewEnumerator3
    Uniform Sprite Size=45
    UsePositionOffset=false
    Write Material Random=true
   )
   /SetVariables(Particles.VisibilityTag=&SpawnGroup()) {isSetParameters=true, module=SetVariables_#13}
   &SpriteFacingAndAlignment(
    Sprite Alignment=<1 1 0>
    Sprite Facing=&MakeVector(X=0, Y=-0.01, Z=1)
    WriteAlignment=false
    WriteFacing=true
   )
   /SetVariables(
    Particles.MeshIndexChunks:NiagaraInt32=&RandomRangeInt(Minimum=0, Maximum=2)
    Particles.MeshOrientationChunks:Quat4f=&RandomQuaternion()
    Particles.ScaleChunks:Vector3f=&Multiply_VectorByFloat(
     Vector=&RandomRangeVector(Minimum=<0.8 0.8 0.8>, Maximum=<1 1 1>)
     Float=&RandomRangeFloat(Minimum=0.03, Maximum=0.04)
    )
   ) {isSetParameters=true, module=SetVariables_#14}
   &SocketLocation(
    Orientation Sampling=::NewEnumerator1
    Position Sampling=::NewEnumerator1
    SamplePosOffset=false
    Socket Index=4
    Socket Reader=$System.Socket Reader
    Socket Sampling Mode=::NewEnumerator3
    VelocitySpeedLimit=false
   ) {module=SocketLocation001}
   &ShapeLocation(
    MaskSpawnGroup=false
    Non Uniform Scale=<4 4 1>
    Offset=&RotateVectorByQuaternion(
     Vector To Rotate=&AddVectorToPosition(Position=Linked() {linkedVariable {name=Emitter.StepPosition, type=Vector3f}}, Vector=<13 -10 0>)
     Quaternion=$Output.SocketLocation001.SampledRotation
    )
    Offset Mode=::NewEnumerator0
    Shape Origin=$Output.SocketLocation001.SampledPosition
    Sphere Distribution=::NewEnumerator3
    Sphere Radius=3
    UseOwnerScale=true
   ) {module=ShapeLocation001}
   /SetVariables(Particles.Position=&MaskVectorBySpawnGroup(
     Masked Vector=&AddVectorToPosition(
      Position=$Output.SocketLocation001.SampledPosition
      Vector=&RotateVectorByQuaternion(
       Vector To Rotate=&AddVectorToPosition(
        Position=Linked() {linkedVariable {name=Emitter.StepPosition, type=Vector3f}}
        Vector=&SelectVectorFromArray(
         Direct Array Index=$Particles.PartitionParticles.Partition
         Vector Selection Array=DataInterface() {propertyValues {FloatData [<20 -7 11>, <24 -7 0>, <20 -7 -11>]}}
        )
       )
       Quaternion=$Output.SocketLocation001.SampledRotation
      )
     )
     Passthrough Vector=Linked() {linkedVariable {name=Particles.Position, type=NiagaraPosition}}
     Spawn Group Mask=3
    )) {isSetParameters=true}
   &AddVelocity(
    Module.Use Falloff Distance=false
    Origin Offset=&RotateVectorByQuaternion(
     Vector To Rotate=&AddVectorToPosition(Position=Linked() {linkedVariable {name=Emitter.StepPosition, type=Vector3f}}, Vector=<-5 10 0>)
     Quaternion=$Output.SocketLocation001.SampledRotation
    )
    Override Seed=false
    Velocity Mode=::NewEnumerator1
    Velocity Origin=$Output.SocketLocation001.SampledPosition
    Velocity Speed=&Multiply_Float(
     A=&RandomRangeFloat(Minimum=0.5, Maximum=1)
     B=&MaskFloat(
      Mask Parameter=$Emitter.SpawnGroup
      Mask=4
      Masked Float=&RandomRangeFloat(Minimum=30, Maximum=100)
      Passthrough Float=220
     )
    )
   )
   &CurlNoiseForce(
    Bias Noise=true
    Bias Noise Field:Vector3f=&RotateVectorByQuaternion(Vector To Rotate=<0.1 -0.3 0>, Quaternion=$Output.SocketLocation001.SampledRotation)
    Noise Frequency=300
    Noise Strength=&MaskFloat(
     Mask Parameter=$Emitter.SpawnGroup
     Mask=4
     Masked Float=&RandomRangeFloat(Minimum=40, Maximum=100)
     Passthrough Float=0
    )
    Pan Noise=true
    Pan Noise Field=<0 0.1 0>
   ) {enabled=false}
   &CurlNoiseForce(
    Bias Noise=true
    Bias Noise Field:Vector3f=&RotateVectorByQuaternion(
     Vector To Rotate=&RandomRangeVector(Minimum=<0.1 -0.1 0>, Maximum=<0.4 -0.2 0>)
     Quaternion=$Output.SocketLocation001.SampledRotation
    )
    Noise Frequency=300
    Noise Strength=&MaskFloat(
     Mask Parameter=$Emitter.SpawnGroup
     Mask=4
     Masked Float=&RandomRangeFloat(Minimum=40, Maximum=100)
     Passthrough Float=0
    )
    Pan Noise=true
    Pan Noise Field=<0 0.1 0>
   ) {module=CurlNoiseForce001}
   &InitialMeshOrientation(Mesh Orientation Mode=::NewEnumerator2, Orientation Vector=<0 0 1>, Rotation=<0.25 0 0>, Use Rotation Vector=true)
   &ApplyInitialForces()
  ]
  particleUpdateScript [
   &SocketLocation(
    Orientation Sampling=::NewEnumerator1
    Position Sampling=::NewEnumerator1
    SamplePosOffset=false
    Socket Index=4
    Socket Reader=$System.Socket Reader
    Socket Sampling Mode=::NewEnumerator3
    VelocitySpeedLimit=false
   )
   /SetVariables(Particles.FXPositionOffset:NiagaraPosition=&AddVectorToPosition(
     Position=$Output.SocketLocation.SampledPosition
     Vector=&RotateVectorByQuaternion(Vector To Rotate=$Emitter.StepPosition, Quaternion=$Output.SocketLocation.SampledRotation)
    )) {isSetParameters=true, module=SetVariables_#15}
   &ParticleState()
   &ScaleSpriteSize(Uniform Curve Sprite Scale=DataInterface() {propertyValues {Curve {keys [
        {arriveTangent=8.5804, interpMode=RCIM_Cubic, leaveTangent=8.5804, tangentMode=RCTM_User, time=0, value=0.1}
        {interpMode=RCIM_Cubic, tangentMode=RCTM_User, time=0.2, value=1}
       ]}}})
   &ScaleColor(
    Scale Alpha=&FloatFromCurve(FloatCurve=DataInterface() {propertyValues {Curve {keys [
         {arriveTangent=4.5659, interpMode=RCIM_Cubic, leaveTangent=4.5659, tangentMode=RCTM_User, time=0, value=0}
         {arriveTangent=0.0464, leaveTangent=0.0464, time=0.1, value=1}
         {arriveTangent=-1.712, interpMode=RCIM_Cubic, leaveTangent=-2.3018, tangentMode=RCTM_Break, time=0.35, value=1}
         {arriveTangent=-0.1251, interpMode=RCIM_Cubic, leaveTangent=-0.1251, tangentMode=RCTM_User, time=1, value=0}
        ]}}})
    Scale RGB=&MakeVectorFromLinearColorRGB(Color=&ColorFromCurve(ColorCurve=DataInterface() {propertyValues {
        AlphaCurve {keys [<at 0 1 RCIM_Cubic>]}
        BlueCurve {keys [<at 0.474 0.7 RCIM_Cubic>,<at 0.6955 0.1 RCIM_Cubic>]}
        GreenCurve {keys [<at 0.474 0.7 RCIM_Cubic>,<at 0.6955 0.1 RCIM_Cubic>]}
        RedCurve {keys [<at 0.474 0.7 RCIM_Cubic>,<at 0.6955 0.1 RCIM_Cubic>]}
       }}))
    ScaleA=true
    ScaleRGB=true
   )
   &Drag(Drag=&RandomRangeFloat(Minimum=1, Maximum=5), Use Linear Drag=true, Use Rotational Drag=true)
   &SolveForcesAndVelocity(Clamp Velocity=false, Limit Acceleration=false)
   /SetVariables(
    Particles.DynParFootprint:Vector4f=&MakeVector4(
     X=&FloatFromCurve(FloatCurve=DataInterface() {propertyValues {Curve {keys [
          {arriveTangent=-8.1467, interpMode=RCIM_Cubic, leaveTangent=-8.1467, tangentMode=RCTM_User, time=0, value=1}
          {arriveTangent=-0.8705, interpMode=RCIM_Cubic, leaveTangent=-0.8705, tangentMode=RCTM_User, time=0.15, value=0.1}
          {arriveTangent=0.136, interpMode=RCIM_Cubic, leaveTangent=0.136, tangentMode=RCTM_User, time=0.2, value=0}
          {arriveTangent=0.2628, interpMode=RCIM_Cubic, leaveTangent=0.2628, tangentMode=RCTM_User, time=0.33, value=0}
          {arriveTangent=2.9851, interpMode=RCIM_Cubic, leaveTangent=2.9851, time=0.5, value=0.25}
          {arriveTangent=-0.0282, interpMode=RCIM_Cubic, leaveTangent=-0.0282, tangentMode=RCTM_User, time=1, value=2}
         ]}}})
     Y=&FloatFromCurve(
      FloatCurve=DataInterface() {propertyValues {Curve {keys [
          {arriveTangent=0.749, interpMode=RCIM_Cubic, leaveTangent=0.749, tangentMode=RCTM_User, time=0.1, value=0.1}
          {arriveTangent=6.5311, interpMode=RCIM_Cubic, leaveTangent=6.5311, tangentMode=RCTM_User, time=0.2, value=0.3}
          {arriveTangent=-0.4547, interpMode=RCIM_Cubic, leaveTangent=-0.4547, tangentMode=RCTM_User, time=0.33, value=1}
          {arriveTangent=-3.7037, interpMode=RCIM_Cubic, leaveTangent=-3.7037, time=0.4, value=0.5}
          {arriveTangent=-0.7364, interpMode=RCIM_Cubic, leaveTangent=-0.7364, tangentMode=RCTM_User, time=0.6, value=0}
         ]}}}
      Scale Curve=2
     )
     Z=0
     W=0
    )
    Particles.ScaleFootprint:Vector3f=&Multiply_VectorByFloat(
     Vector=<0.2 0.2 0.26>
     Float=&FloatFromCurve(FloatCurve=DataInterface() {propertyValues {Curve {keys [{interpMode=RCIM_Cubic, tangentMode=RCTM_User, time=0, value=0}, {interpMode=RCIM_Cubic, tangentMode=RCTM_User, time=0.1, value=1}]}}})
    )
   ) {isSetParameters=true, module=SetVariables_#16}
   /SetVariables(Particles.PositionFootprint:NiagaraPosition=&AddVectorToPosition(
     Position=$Particles.FXPositionOffset
     Vector=&RotateVectorByQuaternion(
      Vector To Rotate=&MakeVector(
       X=&Lerp_Float(
        A=&Lerp_Float(A=-110, B=-130, Alpha=$System.DiveSprint)
        B=1
        Alpha=&FloatFromCurve(FloatCurve=DataInterface() {propertyValues {Curve {keys [
             {arriveTangent=6.5872, interpMode=RCIM_Cubic, leaveTangent=6.5872, tangentMode=RCTM_User, time=0, value=0}
             {arriveTangent=-0.9166, interpMode=RCIM_Cubic, leaveTangent=-0.9166, tangentMode=RCTM_User, time=0.4, value=1}
             {arriveTangent=-0.4877, interpMode=RCIM_Cubic, leaveTangent=-0.4877, tangentMode=RCTM_User, time=0.8, value=0.2}
            ]}}})
       )
       Y=&Lerp_Float(
        A=&Lerp_Float(
         A=-50
         B=-5
         Alpha=&FloatFromCurve(FloatCurve=DataInterface() {propertyValues {Curve {keys [
              <at 0 0.5 RCIM_Cubic>
              {arriveTangent=1.4787, interpMode=RCIM_Cubic, leaveTangent=0.025, tangentMode=RCTM_Break, time=0.13, value=0}
              {
               arriveTangent=11.0317
               arriveTangentWeight=0.4523
               interpMode=RCIM_Cubic
               leaveTangent=2.8039
               leaveTangentWeight=0.1357
               tangentMode=RCTM_Break
               time=0.3
               value=0.9
              }
              {arriveTangent=0.0143, interpMode=RCIM_Cubic, leaveTangent=0.0143, tangentMode=RCTM_Break, time=0.36, value=1}
              {arriveTangent=-3.3431, interpMode=RCIM_Cubic, leaveTangent=-3.3431, tangentMode=RCTM_User, time=0.6, value=0.7}
              {arriveTangent=0.0231, interpMode=RCIM_Cubic, leaveTangent=0.0231, tangentMode=RCTM_User, time=0.85, value=0.4}
             ]}}})
        )
        B=&Lerp_Float(
         A=-50
         B=-5
         Alpha=&FloatFromCurve(FloatCurve=DataInterface() {propertyValues {Curve {keys [
              <at 0 0.5 RCIM_Cubic>
              {arriveTangent=1.4787, interpMode=RCIM_Cubic, leaveTangent=0.025, tangentMode=RCTM_Break, time=0.13, value=0}
              {
               arriveTangent=11.0317
               arriveTangentWeight=0.4523
               interpMode=RCIM_Cubic
               leaveTangent=2.8039
               leaveTangentWeight=0.1357
               tangentMode=RCTM_Break
               time=0.3
               value=0.9
              }
              {arriveTangent=0.0143, interpMode=RCIM_Cubic, leaveTangent=0.0143, tangentMode=RCTM_Break, time=0.36, value=1}
              {arriveTangent=-1.7647, interpMode=RCIM_Cubic, leaveTangent=-1.7647, time=0.5, value=0.875}
              {arriveTangent=-4.2087, interpMode=RCIM_Cubic, leaveTangent=-4.2087, tangentMode=RCTM_User, time=0.7, value=0.4}
             ]}}})
        )
        Alpha=$System.DiveSprint
       )
       Z=0
      )
      Quaternion=$Output.SocketLocation.SampledRotation
     )
    )) {isSetParameters=true, module=SetVariables_#17}
   &UpdateMeshOrientation(
    Facing Direction=&RotateVectorByQuaternion(Vector To Rotate=<1 0 0>, Quaternion=$Output.SocketLocation.SampledRotation)
    Facing Interpolation=false
    Orientation Method=::NewEnumerator1
    Override Quat=false
    TransformFacing=false
    TransformUp=false
    Up Direction:Vector3f=<-1 0 0>
   ) {module=UpdateMeshOrientation001}
   &UpdateMeshOrientation(
    Facing Interpolation=false
    Orientation Method=::NewEnumerator1
    Override Quat=false
    TransformFacing=false
    TransformUp=false
    Up Direction:Vector3f=&MakeVector(
     X=0
     Y=&Lerp_Float(
      A=1
      B=-1
      Alpha=&FloatFromCurve(FloatCurve=DataInterface() {propertyValues {Curve {keys [
           {interpMode=RCIM_Cubic, tangentMode=RCTM_User, time=0.1, value=0}
           {arriveTangent=9.0687, interpMode=RCIM_Cubic, leaveTangent=-0.2697, tangentMode=RCTM_Break, time=0.25, value=1}
           {arriveTangent=-1.7789, interpMode=RCIM_Cubic, leaveTangent=-1.7789, tangentMode=RCTM_User, time=0.33, value=0.6}
           {arriveTangent=0.0864, interpMode=RCIM_Cubic, leaveTangent=0.0864, tangentMode=RCTM_User, time=0.5, value=0.4}
           <at 1 0.7 RCIM_Cubic>
          ]}}})
     )
     Z=&FloatFromCurve(FloatCurve=DataInterface() {propertyValues {Curve {keys [
          {interpMode=RCIM_Cubic, tangentMode=RCTM_User, time=0, value=0.25}
          {interpMode=RCIM_Cubic, tangentMode=RCTM_User, time=0.3, value=1}
          {arriveTangent=-0.5714, interpMode=RCIM_Cubic, leaveTangent=-0.5714, time=0.5, value=0.5}
          {arriveTangent=-0.397, interpMode=RCIM_Cubic, leaveTangent=-0.397, tangentMode=RCTM_User, time=1, value=0.6}
         ]}}})
    )
   ) {module=UpdateMeshOrientation002}
   /SetVariables(Particles.MeshOrientationImpact:Quat4f=&MultiplyQuaternion(
     Quaternion A=&MultiplyQuaternion(Quaternion A=$Output.SocketLocation.SampledRotation, Quaternion B=<1 0 0 1>)
     Quaternion B=<0 0 1 1>
    )) {isSetParameters=true, module=SetVariables_#18}
   /SetVariables(
    Particles.ColorChunks:LinearColor=&ColorFromCurve(ColorCurve=DataInterface() {propertyValues {
       AlphaCurve {keys [<at 0 1 RCIM_Cubic>]}
       BlueCurve {keys [<at 0 1 RCIM_Cubic>,<at 0.1442 0.1812 RCIM_Cubic>]}
       GreenCurve {keys [<at 0 20.8 RCIM_Cubic>,<at 0.1442 0.5693 RCIM_Cubic>]}
       RedCurve {keys [<at 0 100 RCIM_Cubic>,<at 0.1442 1 RCIM_Cubic>]}
      }})
    Particles.PositionRockChunks:NiagaraPosition=&AddVectorToPosition(
     Position=$Particles.FXPositionOffset
     Vector=&AddVectorToPosition(
      Position=&MakeVector(
       X=0
       Y=&Lerp_Float(
        A=0
        B=&Lerp_Float(A=-60, B=-25, Alpha=$System.DiveSprint)
        Alpha=&FloatFromCurve(
         FloatCurve=DataInterface() {propertyValues {Curve {keys [
             {arriveTangent=0.6268, interpMode=RCIM_Cubic, leaveTangent=0.6268, tangentMode=RCTM_User, time=0, value=0}
             {arriveTangent=0.8137, interpMode=RCIM_Cubic, leaveTangent=0.8137, tangentMode=RCTM_User, time=1, value=1}
            ]}}}
         Scale Curve=&RandomRangeFloat(Minimum=0.4, Maximum=1.5)
        )
       )
       Z=&Lerp_Float(
        A=0
        B=&Lerp_Float(A=80, B=230, Alpha=$System.DiveSprint)
        Alpha=&FloatFromCurve(
         FloatCurve=DataInterface() {propertyValues {Curve {keys [
             {arriveTangent=0.3872, interpMode=RCIM_Cubic, leaveTangent=0.3872, tangentMode=RCTM_User, time=0.2, value=0}
             {arriveTangent=1.8857, interpMode=RCIM_Cubic, leaveTangent=1.8857, tangentMode=RCTM_User, time=1, value=1}
            ]}}}
         Scale Curve=&Multiply_Float(
          A=&RandomRangeFloat(Minimum=0.75, Maximum=1)
          B=&FN_FrontEndFloatModifier(Gameplay=1, Front End=0.5)
         )
        )
       )
      )
      Vector=&Multiply_VectorByFloat(
       Vector=&RotateVectorByQuaternion(
        Vector To Rotate=&SelectVectorFromArray(
         Direct Array Index=$Particles.PartitionParticles.Partition
         Vector Selection Array=DataInterface() {propertyValues {FloatData [
            <-0.2995 0.4197 0.0001>
            <-0.3102 0.3246 0.0001>
            <-0.306 0.3354 0.0003>
            <-0.3246 0.1977 0.0003>
            <-0.4562 -0.0267 0.0002>
            <-0.5077 -0.3741 0.0001>
            <-0.4908 -0.4304 0.0003>
            <-0.5093 -0.3564 0.0003>
            <-0.4796 -0.4587 0.0001>
            <-0.4208 -0.4669 0.0001>
            <-0.2903 -0.2606 0.0003>
            <-0.1981 -0.0135 0.0001>
            <-0.1088 0.115 0.0003>
            <-0.099 -0.0415 0.0003>
            <-0.1002 -0.0748 0.0002>
            <-0.1166 -0.2389 0.0002>
            <-0.1085 -0.1962 0.0001>
            <-0.1078 -0.2028 0>
            <-0.1195 -0.3133 0.0002>
            <-0.1064 -0.4026 0.0001>
            <-0.0947 -0.5726 0.0002>
            <-0.0707 -0.6709 0.0003>
            <-0.0505 -0.7363 0.0001>
            <-0.0904 -0.5669 0.0001>
            <-0.0317 -0.8066 0.0003>
            <0.1074 -0.5969 0.0003>
            <0.1375 -0.4617 0.0002>
            <0.1383 -0.4211 0.0003>
            <0.1418 -0.1988 0.0001>
            <0.137 -0.007 0.0003>
            <0.1796 0.0867 0.0002>
            <0.2761 -0.1702 0>
            <0.4498 -0.512 0.0003>
            <0.4742 -0.4529 0.0002>
            <0.4855 -0.4142 0.0003>
            <0.5012 -0.1966 0.0003>
            <0.4866 -0.1118 0>
            <0.3754 0.143 0.0003>
            <0.3499 0.3494 0.0003>
            <0.2809 0.6532 0.0002>
            <0.2329 0.6902 0.0001>
            <0.2263 0.6914 0.0003>
            <0.2093 0.7066 0.0002>
            <0.0516 0.8697 0>
            <-0.1207 0.7843 0.0003>
            <-0.1434 0.764 0.0002>
            <-0.1317 0.7705 0.0002>
            <-0.2533 0.6794 0.0002>
           ]}}
        )
        Quaternion=&MultiplyQuaternion(
         Quaternion A=$Output.SocketLocation.SampledRotation
         Quaternion B=&OrientMeshToVector(Look At Direction=<-1 0 0>, Reference Axis Vector=<1 0 0>)
        )
       )
       Float=&Multiply_Float(
        A=-25
        B=&FloatFromCurve(FloatCurve=DataInterface() {propertyValues {Curve {keys [
             {arriveTangent=1.6232, interpMode=RCIM_Cubic, leaveTangent=1.6232, tangentMode=RCTM_User, time=0.1, value=0.8}
             {interpMode=RCIM_Cubic, tangentMode=RCTM_User, time=0.5, value=1}
            ]}}})
       )
      )
     )
    )
   ) {isSetParameters=true, module=SetVariables_#19}
   /SetVariables(Particles.ScaleChunks:Vector3f=&Multiply_VectorByFloat(
     Vector=$Particles.Initial.ScaleChunks
     Float=&FloatFromCurve(FloatCurve=DataInterface() {propertyValues {Curve {keys [
          {interpMode=RCIM_Cubic, tangentMode=RCTM_User, time=0, value=0}
          {interpMode=RCIM_Cubic, tangentMode=RCTM_User, time=0.14, value=1}
          {arriveTangent=-1.9479, interpMode=RCIM_Cubic, leaveTangent=-1.9479, tangentMode=RCTM_User, time=1, value=0}
         ]}}})
    )) {isSetParameters=true, module=SetVariables_#20}
   /SetVariables(
    Particles.ColorFlames:LinearColor=&ColorFromCurve(ColorCurve=DataInterface() {propertyValues {
       AlphaCurve {keys [<at 0 0>,<at 0.0773 1 RCIM_Cubic>,<at 1 0 RCIM_Cubic>]}
       BlueCurve {keys [<at 0.0656 0.14>,<at 0.4845 0.03>,<at 0.9133 0 RCIM_Cubic>]}
       GreenCurve {keys [<at 0.0656 1.7407>,<at 0.4845 0.1593>,<at 0.9133 0.0083 RCIM_Cubic>]}
       RedCurve {keys [<at 0.0656 7>,<at 0.4845 1>,<at 0.9133 1 RCIM_Cubic>]}
      }})
    Particles.ColorImpact:LinearColor=&ColorFromCurve(ColorCurve=DataInterface() {propertyValues {
       AlphaCurve {keys [<at 0.0107 0>,<at 0.1556 1 RCIM_Cubic>,<at 1 0>]}
       BlueCurve {keys [<at 0.1512 1.7>,<at 1 0 RCIM_Cubic>]}
       GreenCurve {keys [<at 0.1512 1.7>,<at 1 0 RCIM_Cubic>]}
       RedCurve {keys [<at 0.1512 1.7>,<at 1 0 RCIM_Cubic>]}
      }})
    Particles.DynamicMatParFlames:Vector4f=&MakeVector4(
     X=&FloatFromCurve(FloatCurve=DataInterface() {propertyValues {Curve {keys [<at 0 1>,<at 1 0.7>]}}})
     Y=0
     Z=0
     W=0
    )
    Particles.PositionImpact:NiagaraPosition=$Particles.FXPositionOffset
    Particles.ScaleImpact:Vector3f=&Multiply_VectorByFloat(
     Vector=<0.9 1 1>
     Float=&FloatFromCurve(
      FloatCurve=DataInterface() {propertyValues {Curve {keys [
          {arriveTangent=0.897, interpMode=RCIM_Cubic, leaveTangent=0.897, tangentMode=RCTM_User, time=0, value=0.6}
          {interpMode=RCIM_Cubic, tangentMode=RCTM_User, time=0.9, value=1}
         ]}}}
      Scale Curve=0.33
     )
    )
    Particles.SpriteSizeFlames:Vector2f=&Multiply_Vector2D_ByFloat(
     Float=&Multiply_Float(
      A=&RandomRangeFloat(Minimum=37, Maximum=42)
      B=&MergeHelper_FloatSwitch(0=1, 1=1.2, 2=1, 3=?, Index=$Particles.PartitionParticles.Partition, Num=::NewEnumerator1)
     )
     Vector2D=&MakeVector2D(
      X=&FloatFromCurve(FloatCurve=DataInterface() {propertyValues {Curve {keys [<at 0.3 1>,<at 1 0.7>]}}}, Scale Curve=0.8)
      Y=&FloatFromCurve(
       FloatCurve=DataInterface() {propertyValues {Curve {keys [
           {arriveTangent=-0.5687, interpMode=RCIM_Cubic, leaveTangent=-0.5687, tangentMode=RCTM_User, time=0.5, value=1}
           {arriveTangent=-1.6572, interpMode=RCIM_Cubic, leaveTangent=-1.6572, tangentMode=RCTM_User, time=1, value=0.5}
          ]}}}
       Scale Curve=1.4
      )
     )
    )
   ) {isSetParameters=true, module=SetVariables_#21}
   /SetVariables(Particles.ColorImpactCore:LinearColor=&MultiplyLinearColor(
     Linear Color=$Particles.ColorImpact
     Scale Factor=&MakeFloat4FromVectorAndFloat(
      "Vector (RGB)"=&VectorFromFloat(Value=&Lerp_Float(
        A=50
        B=1
        Alpha=&FloatFromCurve(FloatCurve=DataInterface() {propertyValues {Curve {keys [{interpMode=RCIM_Cubic, tangentMode=RCTM_User, time=0.33, value=0}, {interpMode=RCIM_Cubic, tangentMode=RCTM_User, time=0.4, value=1}]}}})
       ))
      "Float (Alpha)"=1
     )
    )) {isSetParameters=true}
   /SetVariables(Particles.DynamicMatParFlameBursts:Vector4f=&MakeVector4(X=&RandomRangeFloat(Minimum=0, Maximum=1), Y=0, Z=0, W=0)) {isSetParameters=true}
   /SetVariables(
    Particles.ColorSparks:LinearColor=&ColorFromCurve(ColorCurve=DataInterface() {propertyValues {
       AlphaCurve {keys [<at 0 0>, {arriveTangent=1.1943, interpMode=RCIM_Cubic, leaveTangent=1.1943, time=0.0773, value=1}, <at 0.8373 1>, <at 1 0 RCIM_Cubic>]}
       BlueCurve {keys [<at 0.0985 0.5>,<at 0.2018 0.12>,<at 0.6092 0>,<at 0.9612 0>]}
       GreenCurve {keys [<at 0.0985 2.3>,<at 0.2018 0.59>,<at 0.6092 0.1>,<at 0.9612 0.0033>]}
       RedCurve {keys [<at 0.0985 5>,<at 0.2018 2>,<at 0.6092 1>,<at 0.9612 0.1>]}
      }})
    Particles.SpriteSizeSparks:Vector2f=&Multiply_Vector2D_ByFloat(
     Float=&RandomRangeFloat(Minimum=10, Maximum=15)
     Vector2D=&Multiply_Vector2D_ByFloat(
      Float=&FloatFromCurve(FloatCurve=DataInterface() {propertyValues {Curve {keys [
           {interpMode=RCIM_Cubic, tangentMode=RCTM_User, time=0, value=1}
           {arriveTangent=-2.4141, interpMode=RCIM_Cubic, leaveTangent=-2.4141, tangentMode=RCTM_User, time=1, value=0}
          ]}}})
      Vector2D=&MakeVector2D(X=0.7, Y=&RandomRangeFloat(Minimum=0.8, Maximum=1))
     )
    )
   ) {isSetParameters=true, module=SetVariables_#22}
   /SetVariables(
    Particles.ColorDirt:LinearColor=&ColorFromCurve(ColorCurve=DataInterface() {propertyValues {
       AlphaCurve {keys [<at 0.3044 0>,<at 0.3827 1 RCIM_Cubic>,<at 0.5082 1>,<at 0.84 0>]}
       BlueCurve {keys [<at 0 0.02 RCIM_Cubic>]}
       GreenCurve {keys [<at 0 0.025 RCIM_Cubic>]}
       RedCurve {keys [<at 0 0.04 RCIM_Cubic>]}
      }})
    Particles.PositionDirt:NiagaraPosition=&AddVectorToPosition(
     Position=$Particles.PositionImpact
     Vector=&RotateVectorByQuaternion(
      Vector To Rotate=&MakeVector(
       X=&Lerp_Float(
        A=0
        B=-20
        Alpha=&FloatFromCurve(FloatCurve=DataInterface() {propertyValues {Curve {keys [<at 0.4 0>,<at 0.9 1>]}}})
       )
       Y=&Lerp_Float(
        A=0
        B=-5
        Alpha=&FloatFromCurve(FloatCurve=DataInterface() {propertyValues {Curve {keys [<at 0.4 0>,<at 0.7 1>]}}})
       )
       Z=0
      )
      Quaternion=$Output.SocketLocation.SampledRotation
     )
    )
    Particles.SpriteSizeDirt:Vector2f=&Multiply_Vector2D_ByFloat(
     Float=&FloatFromCurve(FloatCurve=DataInterface() {propertyValues {Curve {keys [
          <at 0.3 0 RCIM_Cubic>
          {arriveTangent=3.8859, interpMode=RCIM_Cubic, leaveTangent=3.8859, tangentMode=RCTM_User, time=0.4, value=0.4}
          {interpMode=RCIM_Cubic, tangentMode=RCTM_User, time=0.7, value=1}
         ]}}})
     Vector2D=&Vector2DFromFloat(Value=&RandomRangeFloat(Minimum=70, Maximum=85))
    )
   ) {isSetParameters=true, module=SetVariables_#23}
   /SetVariables(
    Particles.ColorTrail:LinearColor=&MultiplyLinearColor(
     Linear Color=<rgba 1 1 1 1>
     Scale Factor=&MakeVector4(
      X=1
      Y=1
      Z=1
      W=&FloatFromCurve(
       FloatCurve=DataInterface() {propertyValues {Curve {keys [
           {interpMode=RCIM_Cubic, tangentMode=RCTM_User, time=0.05, value=0}
           {interpMode=RCIM_Cubic, tangentMode=RCTM_User, time=0.11, value=1}
           {interpMode=RCIM_Cubic, tangentMode=RCTM_User, time=0.14, value=1}
           {interpMode=RCIM_Cubic, tangentMode=RCTM_User, time=0.16, value=0}
           {interpMode=RCIM_Cubic, tangentMode=RCTM_User, time=0.2, value=0}
           {interpMode=RCIM_Cubic, tangentMode=RCTM_User, time=0.28, value=1.4}
           {interpMode=RCIM_Cubic, tangentMode=RCTM_User, time=0.33, value=1}
           {interpMode=RCIM_Cubic, tangentMode=RCTM_User, time=0.5, value=0}
          ]}}}
       Scale Curve=&FN_FrontEndFloatModifier(Gameplay=1.7, Front End=1)
      )
     )
    )
    Particles.DynamicMatParTrail:Vector4f=&MakeVector4(
     X=1
     Y=&FloatFromCurve(
      FloatCurve=DataInterface() {propertyValues {Curve {keys [
          {arriveTangent=-12.4698, interpMode=RCIM_Cubic, leaveTangent=-12.4698, tangentMode=RCTM_User, time=0.02, value=0.65}
          {interpMode=RCIM_Cubic, tangentMode=RCTM_User, time=0.15, value=0.1}
          {arriveTangent=-1.0237, interpMode=RCIM_Cubic, leaveTangent=-1.0237, tangentMode=RCTM_User, time=0.2, value=0.25}
          {arriveTangent=-0.4703, interpMode=RCIM_Cubic, leaveTangent=-0.4703, tangentMode=RCTM_User, time=0.5, value=0.05}
         ]}}}
      Scale Curve=&FN_FrontEndFloatModifier(Gameplay=-0.33, Front End=-0.3)
     )
     Z=0
     W=0
    )
    Particles.ScaleTrail:Vector3f=&Multiply_Vector(
     A=$Particles.ScaleFootprint
     B=&MakeVector(
      X=1
      Y=1
      Z=&FloatFromCurve(
       FloatCurve=DataInterface() {propertyValues {Curve {keys [
           {arriveTangent=6.6357, interpMode=RCIM_Cubic, leaveTangent=6.6357, tangentMode=RCTM_User, time=0, value=0.2}
           {arriveTangent=-2.8877, interpMode=RCIM_Cubic, leaveTangent=-2.8877, tangentMode=RCTM_User, time=0.1, value=1}
           {interpMode=RCIM_Cubic, tangentMode=RCTM_User, time=0.17, value=0}
           {arriveTangent=-0.0918, interpMode=RCIM_Cubic, leaveTangent=-0.0918, tangentMode=RCTM_User, time=0.28, value=1}
           {arriveTangent=-5.8824, interpMode=RCIM_Cubic, leaveTangent=-5.8824, time=0.37, value=0.2}
           {interpMode=RCIM_Cubic, tangentMode=RCTM_User, time=0.45, value=0}
          ]}}}
       Scale Curve=&Lerp_Float(A=2, B=2, Alpha=$System.DiveSprint)
      )
     )
    )
    Particles.TrailOrientation:Quat4f=&MultiplyQuaternion(
     Quaternion A=$Particles.MeshOrientation
     Quaternion B=&MakeCustomQuaternionFromBool(
      Bool:NiagaraBool=&SetBoolByFloatComparison(
       A=&FloatFromCurve(FloatCurve=DataInterface() {propertyValues {Curve {keys [<at 0.1 1>,<at 0.3 0>,<at 1 0>]}}})
       B=0.5
      )
      True Quat:Quat4f=&MultiplyQuaternion(Quaternion A=<0 1 0 1>, Quaternion B=<0 1 0 1>)
      False Quat:Quat4f=<0 0 0 1>
     )
    )
   ) {isSetParameters=true}
   /SetVariables(
    Particles.ColorTrail:LinearColor=&MultiplyLinearColor(
     Linear Color=<rgba 1 1 1 1>
     Scale Factor=&MakeVector4(
      X=1
      Y=1
      Z=1
      W=&FloatFromCurve(
       FloatCurve=DataInterface() {propertyValues {Curve {keys [
           {interpMode=RCIM_Cubic, tangentMode=RCTM_User, time=0.05, value=0}
           {interpMode=RCIM_Cubic, tangentMode=RCTM_User, time=0.1, value=1}
           {interpMode=RCIM_Cubic, tangentMode=RCTM_User, time=0.15, value=1}
           {interpMode=RCIM_Cubic, tangentMode=RCTM_User, time=0.2, value=0}
           {interpMode=RCIM_Cubic, tangentMode=RCTM_User, time=0.21, value=0}
           {interpMode=RCIM_Cubic, tangentMode=RCTM_User, time=0.25, value=1.4}
           {interpMode=RCIM_Cubic, tangentMode=RCTM_User, time=0.33, value=1}
           {interpMode=RCIM_Cubic, tangentMode=RCTM_User, time=0.5, value=0}
          ]}}}
       Scale Curve=&FN_FrontEndFloatModifier(Gameplay=1.7, Front End=1)
      )
     )
    )
    Particles.DynamicMatParTrail:Vector4f=&MakeVector4(
     X=1
     Y=&FloatFromCurve(
      FloatCurve=DataInterface() {propertyValues {Curve {keys [
          {arriveTangent=-12.4698, interpMode=RCIM_Cubic, leaveTangent=-12.4698, tangentMode=RCTM_User, time=0.02, value=0.65}
          {interpMode=RCIM_Cubic, tangentMode=RCTM_User, time=0.15, value=0.1}
          {arriveTangent=-1.0237, interpMode=RCIM_Cubic, leaveTangent=-1.0237, tangentMode=RCTM_User, time=0.2, value=0.25}
          {arriveTangent=-0.4703, interpMode=RCIM_Cubic, leaveTangent=-0.4703, tangentMode=RCTM_User, time=0.5, value=0.05}
         ]}}}
      Scale Curve=&FN_FrontEndFloatModifier(Gameplay=-0.33, Front End=-0.3)
     )
     Z=0
     W=0
    )
    Particles.ScaleTrail:Vector3f=&Multiply_Vector(
     A=$Particles.ScaleFootprint
     B=&MakeVector(
      X=1
      Y=1
      Z=&FloatFromCurve(
       FloatCurve=DataInterface() {propertyValues {Curve {keys [
           {arriveTangent=6.6357, interpMode=RCIM_Cubic, leaveTangent=6.6357, tangentMode=RCTM_User, time=0, value=0.2}
           {arriveTangent=-2.8877, interpMode=RCIM_Cubic, leaveTangent=-2.8877, tangentMode=RCTM_User, time=0.1, value=1}
           {interpMode=RCIM_Cubic, tangentMode=RCTM_User, time=0.17, value=0}
           {arriveTangent=11.25, interpMode=RCIM_Cubic, leaveTangent=11.25, time=0.22, value=0.1}
           {arriveTangent=-7.7393, interpMode=RCIM_Cubic, leaveTangent=-7.7393, tangentMode=RCTM_User, time=0.25, value=0.9}
           {interpMode=RCIM_Cubic, tangentMode=RCTM_User, time=0.45, value=0}
          ]}}}
       Scale Curve=&Lerp_Float(A=2, B=2, Alpha=$System.DiveSprint)
      )
     )
    )
    Particles.TrailOrientation:Quat4f=&MultiplyQuaternion(
     Quaternion A=$Particles.MeshOrientation
     Quaternion B=&MakeCustomQuaternionFromBool(
      Bool:NiagaraBool=&SetBoolByFloatComparison(
       A=&FloatFromCurve(FloatCurve=DataInterface() {propertyValues {Curve {keys [<at 0.1 1>,<at 0.3 0>,<at 1 0>]}}})
       B=0.5
      )
      True Quat:Quat4f=&MultiplyQuaternion(Quaternion A=<0 1 0 1>, Quaternion B=<0 1 0 1>)
      False Quat:Quat4f=<0 0 0 1>
     )
    )
   ) {enabled=false, isSetParameters=true}
  ]
  rendererClasses [&NiagaraMeshRendererProperties, &NiagaraSpriteRendererProperties]
  renderers [
   {rendererClass=&NiagaraMeshRendererProperties, rendererIndex=0}
   {rendererClass=&NiagaraMeshRendererProperties, rendererIndex=1}
   {rendererClass=&NiagaraSpriteRendererProperties, rendererIndex=2}
   {rendererClass=&NiagaraSpriteRendererProperties, rendererIndex=3}
   {rendererClass=&NiagaraMeshRendererProperties, rendererIndex=4}
   {rendererClass=&NiagaraMeshRendererProperties, rendererIndex=5}
   {rendererClass=&NiagaraSpriteRendererProperties, rendererIndex=6}
   {rendererClass=&NiagaraMeshRendererProperties, rendererIndex=7}
  ]
 }
 Fountain {
  bEnabled=false
  emitterUpdateScript [
   &EmitterState(Life Cycle Mode=::NewEnumerator1, Loop Duration=2, UseLoopDelay=false)
   &SpawnRate(SpawnRate=&Lerp_Float(A=3, B=10, Alpha=$User.ParaGlideLeanAlpha), Use Spawn Probability=false)
  ]
  particleSpawnScript [
   &InitializeParticle(
    Lifetime Max=1
    Lifetime Mode=::NewEnumerator1
    Mass Mode=::NewEnumerator3
    Sprite Rotation Mode=::NewEnumerator2
    Sprite Size Mode=::NewEnumerator1
    Uniform Sprite Size Max=12
    Uniform Sprite Size Min=6
    UsePositionOffset=false
    Write Material Random=true
   )
   &AddVelocity(Velocity=&RotateVectorByQuaternion(Vector To Rotate=<0 -100 0>, Quaternion=$Output.SocketLocation.SampledRotation)) {enabled=false}
   &InitialMeshOrientation(Mesh Orientation Mode=::NewEnumerator2, Orientation Vector=<0 0 1>, Rotation=<0.25 0 0>, Use Rotation Vector=true)
   &SocketLocation(
    Orientation Sampling=::NewEnumerator1
    SamplePosOffset=false
    Socket Index=4
    Socket Reader=$System.Socket Reader
    Socket Sampling Mode=::NewEnumerator3
    VelocitySpeedLimit=false
   ) {module=SocketLocation001}
  ]
  particleUpdateScript [
   &SocketLocation(
    Orientation Sampling=::NewEnumerator1
    Position Sampling=::NewEnumerator1
    SamplePosOffset=false
    Socket Index=4
    Socket Reader=$System.Socket Reader
    Socket Sampling Mode=::NewEnumerator3
    VelocitySpeedLimit=false
   )
   &InterpolateOverTime(Rate Of Change=10, Target Quat Value:Quat4f=$Output.SocketLocation.SampledRotation, Variable Type=::NewEnumerator5)
   /SetVariables(Particles.Position=&AddVectorToPosition(
     Position=<0 0 0>
     Vector=&RotateVectorByQuaternion(
      Vector To Rotate=&MakeVector(
       X=&Add_Float(
        A=100
        B=&Lerp_Float(
         A=0
         B=100
         Alpha=&FloatFromCurve(FloatCurve=DataInterface() {propertyValues {Curve {keys [<at 0 0>,<at 1 1>]}}})
        )
       )
       Y=-100
       Z=0
      )
      Quaternion=$Particles.InterpolateOverTime.InterpolateOverTime_Quaternion.Moving Average
     )
    )) {isSetParameters=true}
   &ParticleState()
   &UpdateMeshOrientation(
    Facing Direction=&RotateVectorByQuaternion(Vector To Rotate=<1 0 0>, Quaternion=$Output.SocketLocation.SampledRotation)
    Facing Interpolation=false
    Orientation Method=::NewEnumerator1
    Override Quat=false
    TransformFacing=false
    TransformUp=false
    Up Direction:Vector3f=<1 0 0>
   )
   &SolveForcesAndVelocity(Clamp Velocity=false, Limit Acceleration=false)
  ]
  rendererClasses [&NiagaraSpriteRendererProperties, &NiagaraMeshRendererProperties]
  renderers [
   {rendererClass=&NiagaraSpriteRendererProperties}
   {rendererClass=&NiagaraMeshRendererProperties, rendererIndex=1}
  ]
 }
 Fountain001 {
  bEnabled=false
  emitterUpdateScript [
   &EmitterState(Delay First Loop Only=true, Life Cycle Mode=::NewEnumerator1, Loop Delay=1, Loop Duration=0.5, UseLoopDelay=true)
   &SpawnBurst_Instantaneous(Spawn Count=47, Use Spawn Probability=false, UseLoopCountLimit=false)
   &SpawnRate(SpawnRate=90, Use Spawn Probability=false) {enabled=false}
  ]
  particleSpawnScript [
   &PartitionParticles(Number Of Partitions=47)
   &InitializeParticle(
    Color=<rgba 1 0.4708 0.1462 1>
    Lifetime Max=0.5
    Lifetime Min=0.4
    Lifetime Mode=::NewEnumerator1
    Mass Mode=::NewEnumerator3
    Mesh Scale=&Multiply_VectorByFloat(
     Vector=&RandomRangeVector(Minimum=<0.7 0.7 0.7>, Maximum=<1 1 1>)
     Float=&RandomRangeFloat(Minimum=0.02, Maximum=0.03)
    )
    Mesh Scale Mode=::NewEnumerator4
    Sprite Rotation Mode=::NewEnumerator2
    Sprite Size Mode=::NewEnumerator1
    Uniform Sprite Size Max=3
    Uniform Sprite Size Min=3
    UsePositionOffset=false
    Write Material Random=true
   )
   &SocketLocation(
    Orientation Sampling=::NewEnumerator1
    SamplePosOffset=true
    Sampled Position Offset=&AddVectorToPosition(Position=<0 -30 10>, Vector=<0 0 0>)
    Socket Index=4
    Socket Reader=$System.Socket Reader
    Socket Sampling Mode=::NewEnumerator2
    VelocitySpeedLimit=false
   )
   &InitialMeshOrientation(Use Rotation Vector=false)
  ]
  particleUpdateScript [
   &ParticleState()
   /SetVariables(Particles.Position=&AddVectorToPosition(
     Position=<0 0 0>
     Vector=&Multiply_VectorByFloat(
      Vector=&SelectVectorFromArray(
       Direct Array Index=$Particles.PartitionParticles.Partition
       Vector Selection Array=DataInterface() {propertyValues {FloatData [
          <-0.2995 0.4197 0.0001>
          <-0.3102 0.3246 0.0001>
          <-0.306 0.3354 0.0003>
          <-0.3246 0.1977 0.0003>
          <-0.4562 -0.0267 0.0002>
          <-0.5077 -0.3741 0.0001>
          <-0.4908 -0.4304 0.0003>
          <-0.5093 -0.3564 0.0003>
          <-0.4796 -0.4587 0.0001>
          <-0.4208 -0.4669 0.0001>
          <-0.2903 -0.2606 0.0003>
          <-0.1981 -0.0135 0.0001>
          <-0.1088 0.115 0.0003>
          <-0.099 -0.0415 0.0003>
          <-0.1002 -0.0748 0.0002>
          <-0.1166 -0.2389 0.0002>
          <-0.1085 -0.1962 0.0001>
          <-0.1078 -0.2028 0>
          <-0.1195 -0.3133 0.0002>
          <-0.1064 -0.4026 0.0001>
          <-0.0947 -0.5726 0.0002>
          <-0.0707 -0.6709 0.0003>
          <-0.0505 -0.7363 0.0001>
          <-0.0904 -0.5669 0.0001>
          <-0.0317 -0.8066 0.0003>
          <0.1074 -0.5969 0.0003>
          <0.1375 -0.4617 0.0002>
          <0.1383 -0.4211 0.0003>
          <0.1418 -0.1988 0.0001>
          <0.137 -0.007 0.0003>
          <0.1796 0.0867 0.0002>
          <0.2761 -0.1702 0>
          <0.4498 -0.512 0.0003>
          <0.4742 -0.4529 0.0002>
          <0.4855 -0.4142 0.0003>
          <0.5012 -0.1966 0.0003>
          <0.4866 -0.1118 0>
          <0.3754 0.143 0.0003>
          <0.3499 0.3494 0.0003>
          <0.2809 0.6532 0.0002>
          <0.2329 0.6902 0.0001>
          <0.2263 0.6914 0.0003>
          <0.2093 0.7066 0.0002>
          <0.0516 0.8697 0>
          <-0.1207 0.7843 0.0003>
          <-0.1434 0.764 0.0002>
          <-0.1317 0.7705 0.0002>
          <-0.2533 0.6794 0.0002>
         ]}}
      )
      Float=20
     )
    )) {isSetParameters=true}
   &ScaleMeshSize(Scale Factor=&VectorFromFloat(Value=&FloatFromCurve(FloatCurve=DataInterface() {propertyValues {Curve {keys [
          {interpMode=RCIM_Cubic, tangentMode=RCTM_User, time=0, value=0}
          {interpMode=RCIM_Cubic, tangentMode=RCTM_User, time=0.15, value=1}
          {arriveTangent=0.0702, interpMode=RCIM_Cubic, leaveTangent=0.0702, tangentMode=RCTM_User, time=0.7, value=1}
          {arriveTangent=-2.4412, interpMode=RCIM_Cubic, leaveTangent=-2.4412, tangentMode=RCTM_User, time=1, value=0}
         ]}}})))
   &SolveForcesAndVelocity(Clamp Velocity=false, Limit Acceleration=false)
  ]
  rendererClasses [&NiagaraSpriteRendererProperties, &NiagaraMeshRendererProperties]
  renderers [
   {rendererClass=&NiagaraSpriteRendererProperties}
   {rendererClass=&NiagaraMeshRendererProperties, rendererIndex=1}
  ]
 }
 HandChunks {
  emitterUpdateScript [
   &EmitterState(
    Inactive Response=::NewEnumerator1
    Life Cycle Mode=::NewEnumerator1
    Loop Behavior=::NewEnumerator1
    Loop Duration Mode=::NewEnumerator1
    UseLoopDelay=false
   )
   &SpawnRate(Spawn Probability=0.8, SpawnRate=10, Use Spawn Probability=true)
   &SpawnBurst_Instantaneous(Spawn Count=4, Spawn Probability=HlslExpression() {hlslExpression="Emitter.LoopCount == 0"}, Use Spawn Probability=false) {enabled=false, module=SpawnBurst_Instantaneous001}
  ]
  particleSpawnScript [
   &FN_SetSpawnGroup()
   &PartitionParticles(Number Of Partitions=4)
   &InitializeParticle(
    Color=&RandomRangeLinearColor(Minimum=<rgba 1 0.4257 0.3 1>, Maximum=<rgba 1 0.4667 0.2 1>)
    Lifetime=&Multiply_Float(
     A=&RandomRangeFloat(Minimum=1, Maximum=2)
     B=&FN_FrontEndFloatModifier(Gameplay=1, Front End=&Lerp_Float(A=1, B=0.65, Alpha=$System.DiveSprint))
    )
    Mesh Scale Mode=::NewEnumerator3
    Mesh Uniform Scale=&Multiply_Float(
     A=&RandomRangeFloat(Minimum=0.03, Maximum=0.06)
     B=&MergeHelper_FloatSwitch(0=1, 1=1, 2=0.7, 3=0.7, Index=$Particles.PartitionParticles.Partition, Num=::NewEnumerator2)
    )
    Sprite Rotation Mode=::NewEnumerator2
    UsePositionOffset=false
    Write Material Random=true
   ) {module=InitializeParticle001}
   &CameraOffset()
   /SetVariables(Particles.MeshIndexChunks:NiagaraInt32=&RandomRangeInt(Minimum=0, Maximum=2)) {isSetParameters=true}
   &InitialMeshOrientation(Use Rotation Vector=false)
  ]
  particleUpdateScript [
   &SocketLocation(
    Orientation Sampling=::NewEnumerator1
    SamplePosOffset=true
    Sampled Position Offset=&AddVectorToPosition(
     Position=&Multiply_VectorByFloat(
      Vector=&RandomRangeVector(Minimum=<-15 -15 0>, Maximum=<15 15 15>)
      Float=&FloatFromCurve(FloatCurve=DataInterface() {propertyValues {Curve {keys [
           {arriveTangent=1.779, interpMode=RCIM_Cubic, leaveTangent=1.779, tangentMode=RCTM_User, time=0, value=0.45}
           {interpMode=RCIM_Cubic, tangentMode=RCTM_User, time=0.3, value=1}
          ]}}})
     )
     Vector=&Multiply_VectorByFloat(
      Vector=&FN_FrontEndVector3Modifier(Gameplay=<0 0 98>, Front End=&Lerp_Vector(A=<0 0 100>, B=<0 -60 120>, Alpha=$System.DiveSprint))
      Float=&FloatFromCurve(
       FloatCurve=DataInterface() {propertyValues {Curve {keys [<at 0 -0.2>,<at 1 1>]}}}
       Scale Curve=&RandomRangeFloat(Minimum=0.6, Maximum=1)
      )
     )
    )
    Sampled Position Offset Coordinate Space:ENiagaraSkelMeshTransforms=::NewEnumerator2
    Socket Index=$Particles.PartitionParticles.Partition
    Socket Reader=$System.Socket Reader
    Socket Sampling Mode=::NewEnumerator3
    VelocitySpeedLimit=false
   )
   &ParticleState()
   &ScaleMeshSize(Scale Factor=&VectorFromFloat(Value=&FloatFromCurve(FloatCurve=DataInterface() {propertyValues {Curve {keys [
          {interpMode=RCIM_Cubic, tangentMode=RCTM_User, time=0, value=0}
          {interpMode=RCIM_Cubic, tangentMode=RCTM_User, time=0.15, value=1}
          {interpMode=RCIM_Cubic, tangentMode=RCTM_User, time=1, value=0}
         ]}}})))
   &UpdateMeshOrientation(
    Rotation Rate=&RandomRangeFloat(Minimum=0.4, Maximum=1)
    Rotation Vector=&RandomRangeVector(Minimum=<-1 -1 -1>, Maximum=<1 1 1>)
   )
   &ScaleColor(
    Scale Alpha=&FloatFromCurve(
     FloatCurve=DataInterface() {propertyValues {Curve {keys [
         {interpMode=RCIM_Cubic, tangentMode=RCTM_User, time=0, value=0}
         {interpMode=RCIM_Cubic, tangentMode=RCTM_User, time=0.4, value=1}
         <at 1 0 RCIM_Cubic>
        ]}}}
     Scale Curve=&RandomRangeFloat(Minimum=0.9, Maximum=1.5)
    )
    Scale RGB=&VectorFromFloat(Value=&FloatFromCurve(FloatCurve=DataInterface() {propertyValues {Curve {keys [<at 0 1>,<at 1 0>]}}}))
    ScaleA=true
    ScaleRGB=true
   )
  ]
  rendererClasses [&NiagaraMeshRendererProperties]
  renderers [{rendererClass=&NiagaraMeshRendererProperties, rendererIndex=0}]
 }
}
summary {
 systemName=NS_InflatoDodo_Contrail
 emitters [
  {emitterName=Dirt_Ribbons, rendererClasses [&NiagaraRibbonRendererProperties]}
  {emitterName=DirtClouds, rendererClasses [&NiagaraSpriteRendererProperties]}
  {
   bEnabled=false
   emitterName=Fountain001
   rendererClasses [&NiagaraSpriteRendererProperties, &NiagaraMeshRendererProperties]
  }
  {emitterName=FootprintLeft, rendererClasses [&NiagaraMeshRendererProperties, &NiagaraSpriteRendererProperties]}
  {
   bEnabled=false
   emitterName=Fountain
   rendererClasses [&NiagaraSpriteRendererProperties, &NiagaraMeshRendererProperties]
  }
  {emitterName=FootprintRight, rendererClasses [&NiagaraMeshRendererProperties, &NiagaraSpriteRendererProperties]}
  {emitterName=HandChunks, rendererClasses [&NiagaraMeshRendererProperties]}
 ]
 userVariables [
  User.bIsFrontEnd {type=NiagaraBool}
  User.bIsFrontEndPreview {type=NiagaraBool}
  User.ParaGlideLeanAlpha {type=NiagaraFloat, defaultValue=0, description=""}
  User.RibbonIDOffset {type=NiagaraInt32}
 ]
}
